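# Physics asset for the dynamic skirt (coat bones) on the medium female Ultralight player body.
# Five capsule chains (left, right, back, back-left, back-right) have tight swing and twist limits and hang from master bones constrained to spine or right thigh.
# Angle-free constraints cross-link lower segments of neighboring chains so the panels stay together.
# Spine, pelvis and leg capsules provide body shapes, and a huge sunken root sphere approximates the ground.
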PhysicsAsset /BRCosmetics/Characters/Player/Female/Medium/Bodies/F_MED_Ultralight/Meshes/F_MED_Ultralight_Dyn_skirt
bodies {
 calf_l [{
   shapeType=Capsule
   length=25.7349
   radius=12.7984
   shapeName=pelvis_capsule
   center=<-20.4947 0.5478 1.403>
   rotation=<rot 86.3784 179.9316 -177.3635>
  }]
 calf_r [{
   shapeType=Capsule
   length=25.7349
   radius=12.7984
   shapeName=pelvis_capsule
   center=<20.4947 -0.5478 -1.403>
   rotation=<rot 86.3784 179.9316 2.6365>
  }]
 dyn_coat_bk_01 [{shapeType=Capsule, length=13.5803, radius=17.5, shapeName=dyn_coat_l_01_capsule, center=<5 0 18>, rotation=<rot 0 0 90>}]
 dyn_coat_bk_02 [{shapeType=Capsule, length=19.5508, radius=19.25, shapeName=dyn_coat_l_01_capsule, center=<5 0 18>, rotation=<rot 0 0 90>}]
 dyn_coat_bk_03 [{shapeType=Capsule, length=24.6311, radius=19.25, shapeName=dyn_coat_l_01_capsule, center=<8 0 18>, rotation=<rot 0 0 90>}]
 dyn_coat_bk_04 [{shapeType=Capsule, length=24.6311, radius=19.25, shapeName=dyn_coat_l_01_capsule, center=<8 0 18>, rotation=<rot 0 0 90>}]
 dyn_coat_bk_05 [{shapeType=Capsule, length=29.1311, radius=19.25, shapeName=dyn_coat_l_01_capsule, center=<8 0 18>, rotation=<rot 0 0 90>}]
 dyn_coat_bk_MSTR [{shapeType=Capsule, length=0.505, radius=0.505, shapeName=dyn_coat_l_MSTR_capsule}]
 dyn_coat_bk_l_01 [{shapeType=Capsule, length=6.4264, radius=18.375, shapeName=dyn_coat_l_01_capsule, center=<5 0 18>, rotation=<rot 0 0 90>}]
 dyn_coat_bk_l_02 [{shapeType=Capsule, length=12.5061, radius=18.375, shapeName=dyn_coat_l_01_capsule, center=<5 0 18>, rotation=<rot 0 0 90>}]
 dyn_coat_bk_l_03 [{shapeType=Capsule, length=16.1311, radius=18.625, shapeName=dyn_coat_l_01_capsule, center=<8 0 18>, rotation=<rot 0 0 90>}]
 dyn_coat_bk_l_04 [{shapeType=Capsule, length=16.7561, radius=19.25, shapeName=dyn_coat_l_01_capsule, center=<8 0 18>, rotation=<rot 0 0 90>}]
 dyn_coat_bk_l_05 [{shapeType=Capsule, length=24.2561, radius=19.25, shapeName=dyn_coat_l_01_capsule, center=<8 0 18>, rotation=<rot 0 0 90>}]
 dyn_coat_bk_l_MSTR [{shapeType=Capsule, length=0.505, radius=0.505, shapeName=dyn_coat_l_MSTR_capsule}]
 dyn_coat_bk_r_01 [{shapeType=Capsule, length=6.4264, radius=18.375, shapeName=dyn_coat_l_01_capsule, center=<5 0 18>, rotation=<rot 0 0 90>}]
 dyn_coat_bk_r_02 [{shapeType=Capsule, length=12.5061, radius=18.375, shapeName=dyn_coat_l_01_capsule, center=<5 0 18>, rotation=<rot 0 0 90>}]
 dyn_coat_bk_r_03 [{shapeType=Capsule, length=16.1311, radius=18.625, shapeName=dyn_coat_l_01_capsule, center=<8 0 18>, rotation=<rot -0.0002 -0.0002 93.9014>}]
 dyn_coat_bk_r_04 [{shapeType=Capsule, length=16.7561, radius=19.25, shapeName=dyn_coat_l_01_capsule, center=<8 0 18>, rotation=<rot -0.0002 -0.0002 93.9014>}]
 dyn_coat_bk_r_05 [{shapeType=Capsule, length=24.2561, radius=19.25, shapeName=dyn_coat_l_01_capsule, center=<8 0 18>, rotation=<rot -0.0002 -0.0002 93.9014>}]
 dyn_coat_bk_r_MSTR [{shapeType=Capsule, length=0.505, radius=0.505, shapeName=dyn_coat_l_MSTR_capsule}]
 dyn_coat_l_01 [{shapeType=Capsule, length=10, radius=17.25, shapeName=dyn_coat_l_01_capsule, center=<5 -10 17>, rotation=<rot 0 -0.0004 71.7049>}]
 dyn_coat_l_02 [{shapeType=Capsule, length=10, radius=17, shapeName=dyn_coat_l_01_capsule, center=<5 -10 17>, rotation=<rot 0 -0.0004 71.7049>}]
 dyn_coat_l_03 [{shapeType=Capsule, length=10, radius=17.75, shapeName=dyn_coat_l_01_capsule, center=<10.3356 -10 17>, rotation=<rot 0 -0.0004 71.7049>}]
 dyn_coat_l_04 [{shapeType=Capsule, length=10, radius=19.25, shapeName=dyn_coat_l_01_capsule, center=<10.3356 -10 17>, rotation=<rot 0 -0.0004 71.7049>}]
 dyn_coat_l_MSTR [{shapeType=Capsule, length=0.505, radius=0.505, shapeName=dyn_coat_l_MSTR_capsule}]
 dyn_coat_r_01 [{shapeType=Capsule, length=8.4, radius=18.75, shapeName=dyn_coat_l_01_capsule, center=<5 10 17>, rotation=<rot 0 -0.0009 95.4841>}]
 dyn_coat_r_02 [{shapeType=Capsule, length=8.125, radius=18.875, shapeName=dyn_coat_l_01_capsule, center=<5 10 17>, rotation=<rot 0 -0.0009 95.4841>}]
 dyn_coat_r_03 [{shapeType=Capsule, length=10, radius=17.75, shapeName=dyn_coat_l_01_capsule, center=<10.3356 10 17>, rotation=<rot 0 -0.0009 95.4842>}]
 dyn_coat_r_04 [{shapeType=Capsule, length=10, radius=19.25, shapeName=dyn_coat_l_01_capsule, center=<10.3356 10 17>, rotation=<rot 0 -0.0009 95.4842>}]
 dyn_coat_r_MSTR [{shapeType=Capsule, length=0.505, radius=0.505, shapeName=dyn_coat_l_MSTR_capsule}]
 foot_l [{shapeType=Capsule, length=8.1678, radius=11.2363, shapeName=foot_r_capsule, center=<-2.2209 5.395 0>, rotation=<rot 0 0 -90>}]
 foot_r [{shapeType=Capsule, length=8.1678, radius=11.2363, shapeName=foot_r_capsule, center=<2.2209 -5.395 0>, rotation=<rot 0 0 90>}]
 pelvis [{shapeType=Capsule, length=5.342, radius=13.146, shapeName=pelvis_capsule, center=<-4.8171 -0.6518 -0.0005>}]
 root [{shapeType=Sphere, radius=500, shapeName=root_sphere, center=<0 0 -491.9244>}]
 spine_01 [{shapeType=Capsule, length=6.4652, radius=10.9987, shapeName=pelvis_capsule, center=<-0.9991 -0.3721 0>}]
 spine_02 [{shapeType=Capsule, length=7.4401, radius=10.4987, shapeName=pelvis_capsule, center=<0.1451 -0.7547 0>}]
 spine_03 [{shapeType=Capsule, length=7.4401, radius=10.4987, shapeName=pelvis_capsule, center=<2.1769 0.258 0>}]
 thigh_l [{
   shapeType=Capsule
   length=19.466
   radius=11.8021
   shapeName=pelvis_capsule
   center=<-23.103 -0.2298 3.2133>
   rotation=<rot 89.0998 179.9987 179.9989>
  }]
 thigh_r [{
   shapeType=Capsule
   length=28.841
   radius=11.65
   shapeName=pelvis_capsule
   center=<20.0508 0.2298 -3.2612>
   rotation=<rot 87.5479 179.9996 -0.0002>
  }]
}
constraints [
 {bone1Name=dyn_coat_l_01, bone2Name=dyn_coat_l_MSTR, swing1LimitDegrees=10, swing2LimitDegrees=30, twistLimitDegrees=2}
 {bone1Name=dyn_coat_l_02, bone2Name=dyn_coat_l_01, swing1LimitDegrees=10, swing2LimitDegrees=30, twistLimitDegrees=2}
 {bone1Name=dyn_coat_l_03, bone2Name=dyn_coat_l_02, swing1LimitDegrees=10, swing2LimitDegrees=30, twistLimitDegrees=2}
 {bone1Name=dyn_coat_l_04, bone2Name=dyn_coat_l_03, swing1LimitDegrees=10, swing2LimitDegrees=30, twistLimitDegrees=2}
 {bone1Name=dyn_coat_r_01, bone2Name=dyn_coat_r_MSTR, swing1LimitDegrees=10, swing2LimitDegrees=30, twistLimitDegrees=5}
 {bone1Name=dyn_coat_r_02, bone2Name=dyn_coat_r_01, swing1LimitDegrees=10, swing2LimitDegrees=30, twistLimitDegrees=2}
 {bone1Name=dyn_coat_r_03, bone2Name=dyn_coat_r_02, swing1LimitDegrees=10, swing2LimitDegrees=30, twistLimitDegrees=2}
 {bone1Name=dyn_coat_r_04, bone2Name=dyn_coat_r_03, swing1LimitDegrees=10, swing2LimitDegrees=30, twistLimitDegrees=2}
 {bone1Name=dyn_coat_bk_01, bone2Name=dyn_coat_bk_MSTR, swing1LimitDegrees=10, swing2LimitDegrees=30, twistLimitDegrees=2}
 {bone1Name=dyn_coat_bk_02, bone2Name=dyn_coat_bk_01, swing1LimitDegrees=10, swing2LimitDegrees=30, twistLimitDegrees=2}
 {bone1Name=dyn_coat_bk_l_01, bone2Name=dyn_coat_bk_l_MSTR, swing1LimitDegrees=10, swing2LimitDegrees=30, twistLimitDegrees=2}
 {bone1Name=dyn_coat_bk_l_02, bone2Name=dyn_coat_bk_l_01, swing1LimitDegrees=10, swing2LimitDegrees=30, twistLimitDegrees=2}
 {bone1Name=dyn_coat_bk_r_01, bone2Name=dyn_coat_bk_r_MSTR, swing1LimitDegrees=10, swing2LimitDegrees=30, twistLimitDegrees=2}
 {bone1Name=dyn_coat_bk_r_02, bone2Name=dyn_coat_bk_r_01, swing1LimitDegrees=10, swing2LimitDegrees=30, twistLimitDegrees=2}
 {
  bone1Name=dyn_coat_l_MSTR
  bone2Name=spine_03
  swing1LimitDegrees=45
  swing1Motion=Free
  swing2LimitDegrees=45
  swing2Motion=Free
  twistLimitDegrees=45
  twistMotion=Free
 }
 {
  bone1Name=dyn_coat_r_MSTR
  bone2Name=thigh_r
  swing1LimitDegrees=45
  swing1Motion=Free
  swing2LimitDegrees=45
  swing2Motion=Free
  twistLimitDegrees=45
  twistMotion=Free
 }
 {
  bone1Name=dyn_coat_bk_MSTR
  bone2Name=spine_01
  swing1LimitDegrees=45
  swing1Motion=Free
  swing2LimitDegrees=45
  swing2Motion=Free
  twistLimitDegrees=45
  twistMotion=Free
 }
 {
  bone1Name=dyn_coat_bk_l_MSTR
  bone2Name=spine_01
  swing1LimitDegrees=45
  swing1Motion=Free
  swing2LimitDegrees=45
  swing2Motion=Free
  twistLimitDegrees=45
  twistMotion=Free
 }
 {
  bone1Name=dyn_coat_bk_r_MSTR
  bone2Name=spine_01
  swing1LimitDegrees=45
  swing1Motion=Free
  swing2LimitDegrees=45
  swing2Motion=Free
  twistLimitDegrees=45
  twistMotion=Free
 }
 {
  bone1Name=thigh_l
  bone2Name=pelvis
  swing1LimitDegrees=45
  swing1Motion=Free
  swing2LimitDegrees=45
  swing2Motion=Free
  twistLimitDegrees=45
  twistMotion=Free
 }
 {
  bone1Name=spine_01
  bone2Name=pelvis
  swing1LimitDegrees=45
  swing1Motion=Free
  swing2LimitDegrees=45
  swing2Motion=Free
  twistLimitDegrees=45
  twistMotion=Free
 }
 {
  bone1Name=spine_02
  bone2Name=spine_01
  swing1LimitDegrees=45
  swing1Motion=Free
  swing2LimitDegrees=45
  swing2Motion=Free
  twistLimitDegrees=45
  twistMotion=Free
 }
 {
  bone1Name=spine_03
  bone2Name=spine_02
  swing1LimitDegrees=45
  swing1Motion=Free
  swing2LimitDegrees=45
  swing2Motion=Free
  twistLimitDegrees=45
  twistMotion=Free
 }
 {bone1Name=dyn_coat_bk_r_05, bone2Name=dyn_coat_bk_r_04, swing1LimitDegrees=10, swing2LimitDegrees=30, twistLimitDegrees=2}
 {bone1Name=dyn_coat_bk_l_05, bone2Name=dyn_coat_bk_l_04, swing1LimitDegrees=10, swing2LimitDegrees=30, twistLimitDegrees=2}
 {
  bone1Name=pelvis
  bone2Name=root
  swing1LimitDegrees=45
  swing1Motion=Free
  swing2LimitDegrees=45
  swing2Motion=Free
  twistLimitDegrees=45
  twistMotion=Free
 }
 {bone1Name=dyn_coat_bk_r_04, bone2Name=dyn_coat_bk_r_03, swing1LimitDegrees=10, swing2LimitDegrees=30, twistLimitDegrees=2}
 {bone1Name=dyn_coat_bk_04, bone2Name=dyn_coat_bk_03, swing1LimitDegrees=10, swing2LimitDegrees=30, twistLimitDegrees=2}
 {bone1Name=dyn_coat_bk_l_04, bone2Name=dyn_coat_bk_l_03, swing1LimitDegrees=10, swing2LimitDegrees=30, twistLimitDegrees=2}
 {bone1Name=dyn_coat_bk_03, bone2Name=dyn_coat_bk_02, swing1LimitDegrees=10, swing2LimitDegrees=30, twistLimitDegrees=2}
 {bone1Name=dyn_coat_bk_r_03, bone2Name=dyn_coat_bk_r_02, swing1LimitDegrees=10, swing2LimitDegrees=30, twistLimitDegrees=2}
 {bone1Name=dyn_coat_bk_l_03, bone2Name=dyn_coat_bk_l_02, swing1LimitDegrees=10, swing2LimitDegrees=30, twistLimitDegrees=2}
 {bone1Name=dyn_coat_bk_05, bone2Name=dyn_coat_bk_04, swing1LimitDegrees=10, swing2LimitDegrees=30, twistLimitDegrees=2}
 {
  bone1Name=thigh_r
  bone2Name=pelvis
  swing1LimitDegrees=45
  swing1Motion=Free
  swing2LimitDegrees=45
  swing2Motion=Free
  twistLimitDegrees=45
  twistMotion=Free
 }
 {
  bone1Name=calf_r
  bone2Name=thigh_r
  swing1LimitDegrees=45
  swing1Motion=Free
  swing2LimitDegrees=45
  swing2Motion=Free
  twistLimitDegrees=45
  twistMotion=Free
 }
 {
  bone1Name=dyn_coat_bk_r_04
  bone2Name=dyn_coat_bk_04
  swing1LimitDegrees=45
  swing1Motion=Free
  swing2LimitDegrees=45
  swing2Motion=Free
  twistLimitDegrees=45
  twistMotion=Free
 }
 {
  bone1Name=foot_r
  bone2Name=calf_r
  swing1LimitDegrees=45
  swing1Motion=Free
  swing2LimitDegrees=45
  swing2Motion=Free
  twistLimitDegrees=45
  twistMotion=Free
 }
 {
  bone1Name=dyn_coat_bk_l_04
  bone2Name=dyn_coat_bk_04
  swing1LimitDegrees=45
  swing1Motion=Free
  swing2LimitDegrees=45
  swing2Motion=Free
  twistLimitDegrees=45
  twistMotion=Free
 }
 {
  bone1Name=dyn_coat_r_04
  bone2Name=dyn_coat_bk_r_05
  swing1LimitDegrees=45
  swing1Motion=Free
  swing2LimitDegrees=45
  swing2Motion=Free
  twistLimitDegrees=45
  twistMotion=Free
 }
 {
  bone1Name=calf_l
  bone2Name=thigh_l
  swing1LimitDegrees=45
  swing1Motion=Free
  swing2LimitDegrees=45
  swing2Motion=Free
  twistLimitDegrees=45
  twistMotion=Free
 }
 {
  bone1Name=foot_l
  bone2Name=calf_l
  swing1LimitDegrees=45
  swing1Motion=Free
  swing2LimitDegrees=45
  swing2Motion=Free
  twistLimitDegrees=45
  twistMotion=Free
 }
 {
  bone1Name=dyn_coat_l_04
  bone2Name=dyn_coat_bk_l_05
  swing1LimitDegrees=45
  swing1Motion=Free
  swing2LimitDegrees=45
  swing2Motion=Free
  twistLimitDegrees=45
  twistMotion=Free
 }
]
