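# Secondary-motion AnimBlueprint for the female medium Ultralight body, driving its coat bones.
# It copies the attached parent's pose, runs two deformation Control Rigs and caches the result.
# CapeCache adds thigh-driven bone offsets, with RigidBody skirt simulation in some variants selected by cape-backbling and cloth flags.
# bRigidBodyEnabled blends between CapeCache and an AnimDynamics, leg-constraint and Look At chain fed by a crouch-detection cache.
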AnimBlueprint /BRCosmetics/Characters/Player/Female/Medium/Bodies/F_MED_Ultralight/Meshes/F_MED_Ultralight_AnimBP
target_skeleton=&/F_MED_Ultralight_Skeleton
graphs [
 AnimGraph {nodes [
   Blend Poses by bool(BlendPose_1=Component To Local(), BlendTime_0=0.2, BlendTime_1=0.5, bActiveValue=Get bRigidBodyEnabled()) {blendType=Linear}
   "Constraint - dyn_coat_r_03 (PPPP)"!(
    ComponentPose
    Alpha=1
    ConstraintWeights_0=0.5345
    ConstraintWeights_1:float=0.2132
    ConstraintWeights_2:float=0.1206
    ConstraintWeights_3:float=0.1317
   ) -> (Pose) {
    alphaBoolBlend {blendOption=Linear}
    boneToModify {boneName=dyn_coat_r_03}
    constraintSetup [
     {targetBone {boneName=thigh_r}}
     {targetBone {boneName=calf_r}}
     {targetBone {boneName=deform_knee_bk_01_r}}
     {targetBone {boneName=deform_knee_fr_01_r}}
    ]
    constraintWeights [1,1,1,1]
   }
   "Constraint - dyn_coat_l_03 (PPPP)"!(
    ComponentPose
    Alpha=1
    ConstraintWeights_0=0.5345
    ConstraintWeights_1:float=0.2132
    ConstraintWeights_2:float=0.1206
    ConstraintWeights_3:float=0.1317
   ) -> (Pose) {
    alphaBoolBlend {blendOption=Linear}
    boneToModify {boneName=dyn_coat_l_03}
    constraintSetup [
     {targetBone {boneName=thigh_l}}
     {targetBone {boneName=calf_l}}
     {targetBone {boneName=deform_knee_bk_01_l}}
     {targetBone {boneName=deform_knee_fr_01_l}}
    ]
    constraintWeights [1,1,1,1]
   }
   Local To Component(LocalPose=Use cached pose 'StartupCache'())
   n1 = Get IsSkyDiving() -> (IsSkyDiving:double)
   "Transform (Modify) Bone - Bone: dyn_coat_l_01"(Alpha=@n1, Translation="0,0,0", Rotation="-20,0,0", Scale="1,1,1") {
    rotationMode=BMM_Additive
    scaleMode=BMM_Additive
    translationMode=BMM_Additive
    alphaBoolBlend {blendOption=Linear}
    alphaScaleBiasClamp {bInterpResult=true, interpSpeedDecreasing=9, interpSpeedIncreasing=9}
    boneToModify {boneName=dyn_coat_l_01}
    rotation {pitch=-20}
   }
   "Transform (Modify) Bone - Bone: dyn_coat_r_01"(Alpha=1, Translation="-3,0,0", Rotation="0,0,0", Scale="1,1,1") {
    rotationMode=BMM_Additive
    scaleMode=BMM_Additive
    translationMode=BMM_Additive
    alphaBoolBlend {blendOption=Linear}
    alphaScaleBiasClamp {bInterpResult=true, interpSpeedDecreasing=9, interpSpeedIncreasing=9}
    boneToModify {boneName=dyn_coat_r_01}
    translation {x=-3}
   }
   n2 = "Transform (Modify) Bone - Bone: dyn_coat_l_MSTR"!(
    ComponentPose=Local To Component(LocalPose=Use cached pose 'DetectedCrouchingCache'())
    bAlphaBoolEnabled:bool=True
    Translation="-2,0,5"
    Rotation="-60,-9,15"
   ) -> (Pose) {
    alphaInputType=Bool
    rotationMode=BMM_Replace
    rotationSpace=BCS_BoneSpace
    scaleMode=BMM_Additive
    scaleSpace=BCS_BoneSpace
    translationMode=BMM_Replace
    translationSpace=BCS_BoneSpace
    alphaBoolBlend {blendOption=Linear}
    boneToModify {boneName=dyn_coat_l_MSTR}
    rotation=<rot -60 -9 15>
    translation {x=-2, z=5}
   }
   n3 = "Transform (Modify) Bone - Bone: dyn_coat_bk_l_MSTR"!(ComponentPose=@n2, bAlphaBoolEnabled:bool=True, Translation="0.104,2.001,2.417", Rotation="-7,-60,12") -> (Pose) {
    alphaInputType=Bool
    rotationMode=BMM_Replace
    rotationSpace=BCS_BoneSpace
    scaleMode=BMM_Additive
    scaleSpace=BCS_BoneSpace
    translationMode=BMM_Replace
    translationSpace=BCS_BoneSpace
    alphaBoolBlend {blendOption=Linear}
    boneToModify {boneName=dyn_coat_bk_l_MSTR}
    rotation=<rot -7 -60 12>
    translation=<0.1039 2.001 2.4168>
   }
   n4 = "Transform (Modify) Bone - Bone: dyn_coat_bk_MSTR"!(ComponentPose=@n3, bAlphaBoolEnabled:bool=True, Translation="0,0,0", Rotation="1.232,-59.500,-20.022") -> (Pose) {
    alphaInputType=Bool
    rotationMode=BMM_Replace
    rotationSpace=BCS_BoneSpace
    scaleMode=BMM_Additive
    scaleSpace=BCS_BoneSpace
    translationMode=BMM_Replace
    translationSpace=BCS_BoneSpace
    alphaBoolBlend {blendOption=Linear}
    boneToModify {boneName=dyn_coat_bk_MSTR}
    rotation=<rot 1.232 -59.5 -20.022>
   }
   n5 = "Transform (Modify) Bone - Bone: dyn_coat_bk_r_MSTR"!(ComponentPose=@n4, bAlphaBoolEnabled:bool=True, Translation="-0.005,-0.003,1.406", Rotation="0.125,-20.000,-0.221") -> (Pose) {
    alphaInputType=Bool
    rotationMode=BMM_Replace
    rotationSpace=BCS_BoneSpace
    scaleMode=BMM_Additive
    scaleSpace=BCS_BoneSpace
    translationMode=BMM_Replace
    translationSpace=BCS_BoneSpace
    alphaBoolBlend {blendOption=Linear}
    boneToModify {boneName=dyn_coat_bk_r_MSTR}
    rotation=<rot 0.125 -20 -0.221>
    translation=<-0.005 -0.003 1.4064>
   }
   n6 = "Transform (Modify) Bone - Bone: dyn_coat_r_MSTR"!(ComponentPose=@n5, bAlphaBoolEnabled:bool=True, Translation="-2.910,0.437,4.721", Rotation="16.866,-60.879,-11.556") -> (Pose) {
    alphaInputType=Bool
    rotationMode=BMM_Replace
    rotationSpace=BCS_BoneSpace
    scaleMode=BMM_Additive
    scaleSpace=BCS_BoneSpace
    translationMode=BMM_Replace
    translationSpace=BCS_BoneSpace
    alphaBoolBlend {blendOption=Linear}
    boneToModify {boneName=dyn_coat_r_MSTR}
    rotation=<rot 16.866 -60.8791 -11.556>
    translation=<-2.91 0.437 4.721>
   }
   n7 = Anim Dynamics - Bone: dyn_coat_l_02(ComponentPose=@n6, Alpha=1, ExternalForce="0,0,0") {
    angularDampingOverride=0.9999
    angularSpringConstant=500
    bAngularSpring=true
    bGravityOverrideInSimSpace=true
    bOverrideAngularDamping=true
    bUseGravityOverride=true
    bUseSphericalLimits=true
    gravityScale=0.5
    simulationSpace=RootRelative
    alphaBoolBlend {blendOption=Linear}
    boundBone {boneName=dyn_coat_l_02}
    componentAppliedLinearAccClamp=<5000 5000 5000>
    componentLinearAccScale=<0.2 0.2 0.4>
    componentLinearVelScale=<0.5 0.5 0.5>
    gravityOverride {z=-980}
    physicsBodyDefinitions [{
      sphereCollisionRadius=0
      boundBone {boneName=dyn_coat_l_02}
      boxExtents=<5 20 20>
      constraintSetup {angularLimitsMax=<1 3 5>, angularLimitsMin=<-1 -10 -15>, angularTarget=<0 0 0>}
      localJointOffset=<5 0 0>
     }]
    relativeSpaceBone {boneName=root}
   }
   n8 = Anim Dynamics - Bone: dyn_coat_l_03(ComponentPose=@n7, Alpha=1, ExternalForce="0,0,0") {
    angularDampingOverride=0.9999
    angularSpringConstant=500
    bAngularSpring=true
    bGravityOverrideInSimSpace=true
    bOverrideAngularDamping=true
    bUseGravityOverride=true
    bUseSphericalLimits=true
    gravityScale=0.5
    simulationSpace=RootRelative
    alphaBoolBlend {blendOption=Linear}
    boundBone {boneName=dyn_coat_l_03}
    componentAppliedLinearAccClamp=<3500 3500 3500>
    componentLinearAccScale=<0.2 0.2 0.4>
    componentLinearVelScale=<0.5 0.5 0.5>
    gravityOverride {z=-980}
    physicsBodyDefinitions [{
      sphereCollisionRadius=0
      boundBone {boneName=dyn_coat_l_03}
      boxExtents=<5 20 20>
      constraintSetup {angularLimitsMax=<1 3 5>, angularLimitsMin=<-1 -10 -15>, angularTarget=<0 0 0>}
      localJointOffset=<5 0 0>
     }]
    relativeSpaceBone {boneName=root}
   }
   n9 = Anim Dynamics - Bone: dyn_coat_l_04(ComponentPose=@n8, Alpha=1, ExternalForce="0,0,0") {
    angularDampingOverride=0.9999
    angularSpringConstant=500
    bAngularSpring=true
    bGravityOverrideInSimSpace=true
    bOverrideAngularDamping=true
    bUseGravityOverride=true
    bUseSphericalLimits=true
    gravityScale=0.5
    simulationSpace=RootRelative
    alphaBoolBlend {blendOption=Linear}
    boundBone {boneName=dyn_coat_l_04}
    componentAppliedLinearAccClamp=<5000 5000 5000>
    componentLinearAccScale=<0.2 0.2 0.4>
    componentLinearVelScale=<0.5 0.5 0.5>
    gravityOverride {z=-980}
    physicsBodyDefinitions [{
      sphereCollisionRadius=0
      boundBone {boneName=dyn_coat_l_04}
      boxExtents=<5 20 20>
      constraintSetup {angularLimitsMax=<1 3 5>, angularLimitsMin=<-1 -10 -15>, angularTarget=<0 0 0>}
      localJointOffset=<5 0 0>
     }]
    relativeSpaceBone {boneName=root}
   }
   n10 = Anim Dynamics - Bone: dyn_coat_bk_l_02(ComponentPose=@n9, Alpha=1, ExternalForce="0,0,0") {
    angularDampingOverride=0.9999
    angularSpringConstant=500
    bAngularSpring=true
    bGravityOverrideInSimSpace=true
    bOverrideAngularDamping=true
    bUseGravityOverride=true
    bUseSphericalLimits=true
    gravityScale=0.5
    simulationSpace=RootRelative
    alphaBoolBlend {blendOption=Linear}
    boundBone {boneName=dyn_coat_bk_l_02}
    componentAppliedLinearAccClamp=<5000 5000 5000>
    componentLinearAccScale=<0.2 0.2 0.4>
    componentLinearVelScale=<0.5 0.5 0.5>
    gravityOverride {z=-980}
    physicsBodyDefinitions [{
      collisionType=CustomSphere
      sphereCollisionRadius=2
      boundBone {boneName=dyn_coat_bk_l_02}
      boxExtents=<8 20 20>
      constraintSetup {angularLimitsMax=<1 2 10>, angularLimitsMin=<-1 -10 -10>, angularTarget=<0 0 0>}
      localJointOffset=<5 0 0>
     }]
    relativeSpaceBone {boneName=root}
   }
   n11 = Anim Dynamics - Bone: dyn_coat_bk_l_04(ComponentPose=@n10, Alpha=1, ExternalForce="0,0,0") {
    angularDampingOverride=0.9999
    angularSpringConstant=500
    bAngularSpring=true
    bGravityOverrideInSimSpace=true
    bOverrideAngularDamping=true
    bUseGravityOverride=true
    bUseSphericalLimits=true
    gravityScale=0.5
    simulationSpace=RootRelative
    alphaBoolBlend {blendOption=Linear}
    boundBone {boneName=dyn_coat_bk_l_04}
    componentAppliedLinearAccClamp=<5000 5000 5000>
    componentLinearAccScale=<0.2 0.2 0.4>
    componentLinearVelScale=<0.5 0.5 0.5>
    gravityOverride {z=-980}
    physicsBodyDefinitions [{
      collisionType=CustomSphere
      sphereCollisionRadius=2
      boundBone {boneName=dyn_coat_bk_l_04}
      boxExtents=<8 20 20>
      constraintSetup {angularLimitsMax=<1 2 10>, angularLimitsMin=<-1 -10 -10>, angularTarget=<0 0 0>}
      localJointOffset=<5 0 0>
     }]
    relativeSpaceBone {boneName=root}
   }
   n12 = Anim Dynamics - Bone: dyn_coat_bk_l_05(ComponentPose=@n11, Alpha=1, ExternalForce="0,0,0") {
    angularDampingOverride=0.9999
    angularSpringConstant=500
    bAngularSpring=true
    bGravityOverrideInSimSpace=true
    bOverrideAngularDamping=true
    bUseGravityOverride=true
    bUseSphericalLimits=true
    gravityScale=0.5
    simulationSpace=RootRelative
    alphaBoolBlend {blendOption=Linear}
    boundBone {boneName=dyn_coat_bk_l_05}
    componentAppliedLinearAccClamp=<3500 3500 3500>
    componentLinearAccScale=<0.2 0.2 0.4>
    componentLinearVelScale=<0.5 0.5 0.5>
    gravityOverride {z=-980}
    physicsBodyDefinitions [{
      collisionType=CustomSphere
      sphereCollisionRadius=2
      boundBone {boneName=dyn_coat_bk_l_05}
      boxExtents=<8 20 20>
      constraintSetup {angularLimitsMax=<1 2 10>, angularLimitsMin=<-1 -10 -10>, angularTarget=<0 0 0>}
      localJointOffset=<5 0 0>
     }]
    relativeSpaceBone {boneName=root}
   }
   n13 = Anim Dynamics - Bone: dyn_coat_bk_03(ComponentPose=@n12, Alpha=1, ExternalForce="0,0,0") {
    angularDampingOverride=0.9999
    angularSpringConstant=500
    bAngularSpring=true
    bGravityOverrideInSimSpace=true
    bOverrideAngularDamping=true
    bUseGravityOverride=true
    bUseSphericalLimits=true
    gravityScale=0.5
    simulationSpace=RootRelative
    alphaBoolBlend {blendOption=Linear}
    boundBone {boneName=dyn_coat_bk_03}
    componentAppliedLinearAccClamp=<5000 5000 5000>
    componentLinearAccScale=<0.2 0.2 0.4>
    componentLinearVelScale=<0.5 0.5 0.5>
    gravityOverride {z=-980}
    physicsBodyDefinitions [{
      collisionType=CustomSphere
      sphereCollisionRadius=2
      boundBone {boneName=dyn_coat_bk_03}
      boxExtents=<8 20 20>
      constraintSetup {angularLimitsMax=<1 2 10>, angularLimitsMin=<-1 -10 -10>, angularTarget=<0 0 0>}
      localJointOffset=<5 0 0>
     }]
    relativeSpaceBone {boneName=root}
   }
   n14 = Anim Dynamics - Bone: dyn_coat_bk_04(ComponentPose=@n13, Alpha=1, ExternalForce="0,0,0") {
    angularDampingOverride=0.9999
    angularSpringConstant=500
    bAngularSpring=true
    bGravityOverrideInSimSpace=true
    bOverrideAngularDamping=true
    bUseGravityOverride=true
    bUseSphericalLimits=true
    gravityScale=0.5
    simulationSpace=RootRelative
    alphaBoolBlend {blendOption=Linear}
    boundBone {boneName=dyn_coat_bk_04}
    componentAppliedLinearAccClamp=<5000 5000 5000>
    componentLinearAccScale=<0.2 0.2 0.4>
    componentLinearVelScale=<0.5 0.5 0.5>
    gravityOverride {z=-980}
    physicsBodyDefinitions [{
      collisionType=CustomSphere
      sphereCollisionRadius=2
      boundBone {boneName=dyn_coat_bk_04}
      boxExtents=<8 20 20>
      constraintSetup {angularLimitsMax=<1 2 10>, angularLimitsMin=<-1 -10 -10>, angularTarget=<0 0 0>}
      localJointOffset=<5 0 0>
     }]
    relativeSpaceBone {boneName=root}
   }
   n15 = Anim Dynamics - Bone: dyn_coat_bk_05(ComponentPose=@n14, Alpha=1, ExternalForce="0,0,0") {
    angularDampingOverride=0.9999
    angularSpringConstant=500
    bAngularSpring=true
    bGravityOverrideInSimSpace=true
    bOverrideAngularDamping=true
    bUseGravityOverride=true
    bUseSphericalLimits=true
    gravityScale=0.5
    simulationSpace=RootRelative
    alphaBoolBlend {blendOption=Linear}
    boundBone {boneName=dyn_coat_bk_05}
    componentAppliedLinearAccClamp=<3500 3500 3500>
    componentLinearAccScale=<0.2 0.2 0.4>
    componentLinearVelScale=<0.5 0.5 0.5>
    gravityOverride {z=-980}
    physicsBodyDefinitions [{
      collisionType=CustomSphere
      sphereCollisionRadius=2
      boundBone {boneName=dyn_coat_bk_05}
      boxExtents=<8 20 20>
      constraintSetup {angularLimitsMax=<1 2 10>, angularLimitsMin=<-1 -10 -10>, angularTarget=<0 0 0>}
      localJointOffset=<5 0 0>
     }]
    relativeSpaceBone {boneName=root}
   }
   n16 = Anim Dynamics - Bone: dyn_coat_bk_r_03(ComponentPose=@n15, Alpha=1, ExternalForce="0,0,0") {
    angularDampingOverride=0.9999
    angularSpringConstant=500
    bAngularSpring=true
    bGravityOverrideInSimSpace=true
    bOverrideAngularDamping=true
    bUseGravityOverride=true
    bUseSphericalLimits=true
    gravityScale=0.5
    simulationSpace=RootRelative
    alphaBoolBlend {blendOption=Linear}
    boundBone {boneName=dyn_coat_bk_r_03}
    componentAppliedLinearAccClamp=<5000 5000 5000>
    componentLinearAccScale=<0.2 0.2 0.4>
    componentLinearVelScale=<0.5 0.5 0.5>
    gravityOverride {z=-980}
    physicsBodyDefinitions [{
      collisionType=CustomSphere
      sphereCollisionRadius=2
      boundBone {boneName=dyn_coat_bk_r_03}
      boxExtents=<8 20 20>
      constraintSetup {angularLimitsMax=<1 2 10>, angularLimitsMin=<-1 -10 -10>, angularTarget=<0 0 0>}
      localJointOffset=<5 0 0>
     }]
    relativeSpaceBone {boneName=root}
   }
   n17 = Anim Dynamics - Bone: dyn_coat_bk_r_04(ComponentPose=@n16, Alpha=1, ExternalForce="0,0,0") {
    angularDampingOverride=0.9999
    angularSpringConstant=500
    bAngularSpring=true
    bGravityOverrideInSimSpace=true
    bOverrideAngularDamping=true
    bUseGravityOverride=true
    bUseSphericalLimits=true
    gravityScale=0.5
    simulationSpace=RootRelative
    alphaBoolBlend {blendOption=Linear}
    boundBone {boneName=dyn_coat_bk_r_04}
    componentAppliedLinearAccClamp=<5000 5000 5000>
    componentLinearAccScale=<0.2 0.2 0.4>
    componentLinearVelScale=<0.5 0.5 0.5>
    gravityOverride {z=-980}
    physicsBodyDefinitions [{
      collisionType=CustomSphere
      sphereCollisionRadius=2
      boundBone {boneName=dyn_coat_bk_r_04}
      boxExtents=<8 20 20>
      constraintSetup {angularLimitsMax=<1 2 10>, angularLimitsMin=<-1 -10 -10>, angularTarget=<0 0 0>}
      localJointOffset=<5 0 0>
     }]
    relativeSpaceBone {boneName=root}
   }
   n18 = Anim Dynamics - Bone: dyn_coat_bk_r_05(ComponentPose=@n17, Alpha=1, ExternalForce="0,0,0") {
    angularDampingOverride=0.9999
    angularSpringConstant=500
    bAngularSpring=true
    bGravityOverrideInSimSpace=true
    bOverrideAngularDamping=true
    bUseGravityOverride=true
    bUseSphericalLimits=true
    gravityScale=0.5
    simulationSpace=RootRelative
    alphaBoolBlend {blendOption=Linear}
    boundBone {boneName=dyn_coat_bk_r_05}
    componentAppliedLinearAccClamp=<3500 3500 3500>
    componentLinearAccScale=<0.2 0.2 0.4>
    componentLinearVelScale=<0.5 0.5 0.5>
    gravityOverride {z=-980}
    physicsBodyDefinitions [{
      collisionType=CustomSphere
      sphereCollisionRadius=2
      boundBone {boneName=dyn_coat_bk_r_05}
      boxExtents=<8 20 20>
      constraintSetup {angularLimitsMax=<1 2 10>, angularLimitsMin=<-1 -10 -10>, angularTarget=<0 0 0>}
      localJointOffset=<5 0 0>
     }]
    relativeSpaceBone {boneName=root}
   }
   n19 = Anim Dynamics - Bone: dyn_coat_r_02(ComponentPose=@n18, Alpha=1, ExternalForce="0,0,0") {
    angularDampingOverride=0.9999
    angularSpringConstant=500
    bAngularSpring=true
    bGravityOverrideInSimSpace=true
    bOverrideAngularDamping=true
    bUseGravityOverride=true
    gravityScale=0.5
    simulationSpace=RootRelative
    alphaBoolBlend {blendOption=Linear}
    boundBone {boneName=dyn_coat_r_02}
    componentAppliedLinearAccClamp=<5000 5000 5000>
    componentLinearAccScale=<0.2 0.2 0.4>
    componentLinearVelScale=<0.5 0.5 0.5>
    gravityOverride {z=-980}
    physicsBodyDefinitions [{
      sphereCollisionRadius=0
      boundBone {boneName=dyn_coat_r_02}
      boxExtents=<5 20 20>
      constraintSetup {angularLimitsMax=<1 1 5>, angularLimitsMin=<-1 -10 -10>, angularTarget=<0 0 0>}
      localJointOffset=<5 0 0>
     }]
    relativeSpaceBone {boneName=root}
   }
   n20 = Anim Dynamics - Bone: dyn_coat_r_03(ComponentPose=@n19, Alpha=1, ExternalForce="0,0,0") {
    angularDampingOverride=0.9999
    angularSpringConstant=500
    bAngularSpring=true
    bGravityOverrideInSimSpace=true
    bOverrideAngularDamping=true
    bUseGravityOverride=true
    gravityScale=0.5
    simulationSpace=RootRelative
    alphaBoolBlend {blendOption=Linear}
    boundBone {boneName=dyn_coat_r_03}
    componentAppliedLinearAccClamp=<3500 3500 3500>
    componentLinearAccScale=<0.2 0.2 0.4>
    componentLinearVelScale=<0.5 0.5 0.5>
    gravityOverride {z=-980}
    physicsBodyDefinitions [{
      sphereCollisionRadius=0
      boundBone {boneName=dyn_coat_r_03}
      boxExtents=<5 20 20>
      constraintSetup {angularLimitsMax=<1 1 5>, angularLimitsMin=<-1 -10 -10>, angularTarget=<0 0 0>}
      localJointOffset=<5 0 0>
     }]
    relativeSpaceBone {boneName=root}
   }
   n21 = Anim Dynamics - Bone: dyn_coat_r_04(ComponentPose=@n20, Alpha=1, ExternalForce="0,0,0") {
    angularDampingOverride=0.9999
    angularSpringConstant=500
    bAngularSpring=true
    bGravityOverrideInSimSpace=true
    bOverrideAngularDamping=true
    bUseGravityOverride=true
    gravityScale=0.5
    simulationSpace=RootRelative
    alphaBoolBlend {blendOption=Linear}
    boundBone {boneName=dyn_coat_r_04}
    componentAppliedLinearAccClamp=<5000 5000 5000>
    componentLinearAccScale=<0.2 0.2 0.4>
    componentLinearVelScale=<0.5 0.5 0.5>
    gravityOverride {z=-980}
    physicsBodyDefinitions [{
      sphereCollisionRadius=0
      boundBone {boneName=dyn_coat_r_04}
      boxExtents=<5 20 20>
      constraintSetup {angularLimitsMax=<1 1 5>, angularLimitsMin=<-1 -10 -10>, angularTarget=<0 0 0>}
      localJointOffset=<5 0 0>
     }]
    relativeSpaceBone {boneName=root}
   }
   n22 = "Constraint - dyn_coat_l_02 (PPPP)"!(
    ComponentPose=@n21
    Alpha=1
    ConstraintWeights_0=0.0017
    ConstraintWeights_1:float=0.1839
    ConstraintWeights_2:float=0.8081
    ConstraintWeights_3:float=0.0063
   ) -> (Pose) {
    alphaBoolBlend {blendOption=Linear}
    boneToModify {boneName=dyn_coat_l_02}
    constraintSetup [{targetBone {boneName=pelvis}}, {targetBone {boneName=thigh_twist_01_l}}, {targetBone {boneName=thigh_l}}, {targetBone {boneName=calf_l}}]
    constraintWeights [1,1,1,1]
   }
   n23 = "Constraint - dyn_coat_l_03 (P)"(ComponentPose=@n22, Alpha=1, ConstraintWeights_0=1) {
    alphaBoolBlend {blendOption=Linear}
    boneToModify {boneName=dyn_coat_l_03}
    constraintSetup [{targetBone {boneName=calf_l}}]
    constraintWeights [1]
   }
   n24 = "Constraint - dyn_coat_bk_l_02 (PPP)"!(ComponentPose=@n23, Alpha=1, ConstraintWeights_0=0.1619, ConstraintWeights_1:float=0.54, ConstraintWeights_2:float=0.2981) -> (Pose) {
    alphaBoolBlend {blendOption=Linear}
    boneToModify {boneName=dyn_coat_bk_l_02}
    constraintSetup [{targetBone {boneName=pelvis}}, {targetBone {boneName=thigh_twist_01_l}}, {targetBone {boneName=thigh_l}}]
    constraintWeights [1,1,1]
   }
   n25 = "Constraint - dyn_coat_bk_l_03 (PPPP)"!(
    ComponentPose=@n24
    Alpha=1
    ConstraintWeights_0=0.0039
    ConstraintWeights_1:float=0.2026
    ConstraintWeights_2:float=0.7898
    ConstraintWeights_3:float=0.0037
   ) -> (Pose) {
    alphaBoolBlend {blendOption=Linear}
    boneToModify {boneName=dyn_coat_bk_l_03}
    constraintSetup [{targetBone {boneName=pelvis}}, {targetBone {boneName=thigh_twist_01_l}}, {targetBone {boneName=thigh_l}}, {targetBone {boneName=calf_l}}]
    constraintWeights [1,1,1,1]
   }
   n26 = "Constraint - dyn_coat_bk_l_04 (PPP)"!(ComponentPose=@n25, Alpha=1, ConstraintWeights_0=0.3342, ConstraintWeights_1:float=0.5002, ConstraintWeights_2:float=0.1656) -> (Pose) {
    alphaBoolBlend {blendOption=Linear}
    boneToModify {boneName=dyn_coat_bk_l_04}
    constraintSetup [{targetBone {boneName=thigh_l}}, {targetBone {boneName=calf_l}}, {targetBone {boneName=deform_knee_bk_01_l}}]
    constraintWeights [1,1,1]
   }
   n27 = "Constraint - dyn_coat_bk_l_04 (P)"(ComponentPose=@n26, Alpha=1, ConstraintWeights_0=1) {
    alphaBoolBlend {blendOption=Linear}
    boneToModify {boneName=dyn_coat_bk_l_04}
    constraintSetup [{targetBone {boneName=calf_l}}]
    constraintWeights [1]
   }
   n28 = "Constraint - dyn_coat_bk_r_02 (PPP)"!(ComponentPose=@n27, Alpha=1, ConstraintWeights_0=0.1254, ConstraintWeights_1:float=0.4789, ConstraintWeights_2:float=0.3956) -> (Pose) {
    alphaBoolBlend {blendOption=Linear}
    boneToModify {boneName=dyn_coat_bk_r_02}
    constraintSetup [{targetBone {boneName=pelvis}}, {targetBone {boneName=thigh_twist_01_r}}, {targetBone {boneName=thigh_r}}]
    constraintWeights [1,1,1]
   }
   n29 = "Constraint - dyn_coat_bk_r_03 (PPPP)"!(
    ComponentPose=@n28
    Alpha=1
    ConstraintWeights_0=0.0039
    ConstraintWeights_1:float=0.2026
    ConstraintWeights_2:float=0.7898
    ConstraintWeights_3:float=0.0037
   ) -> (Pose) {
    alphaBoolBlend {blendOption=Linear}
    boneToModify {boneName=dyn_coat_bk_r_03}
    constraintSetup [{targetBone {boneName=pelvis}}, {targetBone {boneName=thigh_twist_01_r}}, {targetBone {boneName=thigh_r}}, {targetBone {boneName=calf_r}}]
    constraintWeights [1,1,1,1]
   }
   n30 = "Constraint - dyn_coat_bk_r_04 (PPPP)"!(
    ComponentPose=@n29
    Alpha=1
    ConstraintWeights_0=0.5391
    ConstraintWeights_1:float=0.2654
    ConstraintWeights_2:float=0.192
    ConstraintWeights_3:float=0.0035
   ) -> (Pose) {
    alphaBoolBlend {blendOption=Linear}
    boneToModify {boneName=dyn_coat_bk_r_04}
    constraintSetup [
     {targetBone {boneName=thigh_r}}
     {targetBone {boneName=calf_r}}
     {targetBone {boneName=deform_knee_bk_01_r}}
     {targetBone {boneName=deform_knee_fr_01_r}}
    ]
    constraintWeights [1,1,1,1]
   }
   n31 = "Constraint - dyn_coat_bk_r_04 (P)"(ComponentPose=@n30, Alpha=1, ConstraintWeights_0=1) {
    alphaBoolBlend {blendOption=Linear}
    boneToModify {boneName=dyn_coat_bk_r_04}
    constraintSetup [{targetBone {boneName=calf_r}}]
    constraintWeights [1]
   }
   n32 = "Constraint - dyn_coat_r_02 (PPPP)"!(
    ComponentPose=@n31
    Alpha=1
    ConstraintWeights_0=0.0018
    ConstraintWeights_1:float=0.1878
    ConstraintWeights_2:float=0.8042
    ConstraintWeights_3:float=0.0062
   ) -> (Pose) {
    alphaBoolBlend {blendOption=Linear}
    boneToModify {boneName=dyn_coat_r_02}
    constraintSetup [{targetBone {boneName=pelvis}}, {targetBone {boneName=thigh_twist_01_r}}, {targetBone {boneName=thigh_r}}, {targetBone {boneName=calf_r}}]
    constraintWeights [1,1,1,1]
   }
   n33 = "Constraint - dyn_coat_r_03 (P)"(ComponentPose=@n32, Alpha=1, ConstraintWeights_0=1) {
    alphaBoolBlend {blendOption=Linear}
    boneToModify {boneName=dyn_coat_r_03}
    constraintSetup [{targetBone {boneName=calf_r}}]
    constraintWeights [1]
   }
   n34 = Look At - Bone: dyn_coat_bk_l_MSTR(ComponentPose=@n33, Alpha=1) {
    alphaBoolBlend {blendOption=Linear}
    boneToModify {boneName=dyn_coat_bk_l_MSTR}
    lookAtLocation=<-51.9573 -13.1551 -37.7464>
    lookAtTarget {boneReference {boneName=pelvis}}
   }
   n35 = Look At - Bone: dyn_coat_bk_MSTR(ComponentPose=@n34, Alpha=1) {
    alphaBoolBlend {blendOption=Linear}
    boneToModify {boneName=dyn_coat_bk_MSTR}
    lookAtLocation=<17.0987 -69.1867 -12.712>
    lookAtTarget {boneReference {boneName=root}}
   }
   n36 = Look At - Bone: dyn_coat_bk_r_MSTR(ComponentPose=@n35, Alpha=1) {
    alphaBoolBlend {blendOption=Linear}
    boneToModify {boneName=dyn_coat_bk_r_MSTR}
    lookAtLocation=<-109.5739 -35.2655 32.101>
    lookAtTarget {boneReference {boneName=pelvis}}
    lookAt_Axis {axis=<1 0>}
   }
   Output Pose(Result=Blend Poses by bool(
     BlendPose_0=Use cached pose 'CapeCache'()
     BlendPose_1=Component To Local(ComponentPose=@n36)
     BlendTime_0=0.2
     BlendTime_1=0.5
     bActiveValue=Get bRigidBodyEnabled()
    ) {blendType=Linear}) {layerGroup=None}
   n37 = Control Rig(Source=Copy Pose From Mesh() {bUseAttachedParent=true}) {
    alphaBoolBlend {blendOption=Linear}
    controlRigAssetReference {blueprintRigClass=&/F_MED_Ultralight_Deform_CtrlRig_twists.F_MED_Ultralight_Deform_CtrlRig_twists_C}
   }
   n38 = Control Rig!(
    Source=@n37
    Alpha
    KneeRot_fr_l_114deg:Vector
    KneeTrans_fr_l_114deg:Vector="0.258,-3.591,0.009"
    KneeScale_fr_l_114deg:Vector="1,1,1"
    KneeScale_fr_l_130deg:Vector="1,1,1"
    KneeRot_fr_l_130deg:Vector="0,0,-51.2"
    KneeTrans_fr_l_130deg:Vector="0.258,-3.591,0.009"
   ) -> (Pose) {
    alphaBoolBlend {blendOption=Linear}
    controlRigAssetReference {blueprintRigClass=&/F_MED_Ultralight_Deform_CtrlRig_sculpters.F_MED_Ultralight_Deform_CtrlRig_sculpters_C}
   }
   Save cached pose 'StartupCache'(Pose=@n38) {cacheName=StartupCache}
   n39 = Get thigh_rot_up_l() -> (thigh_rot_up_l:double)
   n40 = "Transform (Modify) Bone - Bone: dyn_coat_l_MSTR"!(
    ComponentPose=Local To Component(LocalPose=Use cached pose 'StartupCache'())
    Alpha:float=@n39
    Translation="-6,-2,7"
    Rotation="0,-80,10"
    Scale:Vector="1,1,1"
   ) -> (Pose) {
    rotationMode=BMM_Additive
    rotationSpace=BCS_BoneSpace
    scaleMode=BMM_Additive
    scaleSpace=BCS_BoneSpace
    translationMode=BMM_Additive
    translationSpace=BCS_BoneSpace
    alphaBoolBlend {blendOption=Linear}
    boneToModify {boneName=dyn_coat_l_MSTR}
    rotation {roll=10, yaw=-80}
    translation=<-6 -2 7>
   }
   n41 = "Transform (Modify) Bone - Bone: dyn_coat_bk_l_MSTR"!(ComponentPose=@n40, Alpha:float=@n39, Translation="-6,0,0", Rotation="0,0,0", Scale:Vector="1,1,1") -> (Pose) {
    rotationMode=BMM_Additive
    rotationSpace=BCS_BoneSpace
    scaleMode=BMM_Additive
    scaleSpace=BCS_BoneSpace
    translationMode=BMM_Additive
    translationSpace=BCS_BoneSpace
    alphaBoolBlend {blendOption=Linear}
    boneToModify {boneName=dyn_coat_bk_l_MSTR}
    translation {x=-6}
   }
   n42 = Get thigh_rot_up_r() -> (thigh_rot_up_r:double)
   n43 = "Transform (Modify) Bone - Bone: dyn_coat_r_MSTR"!(ComponentPose=@n41, Alpha:float=@n42, Translation="-6,-2,4.180", Rotation="0,90,-10", Scale:Vector="1,1,1") -> (Pose) {
    rotationMode=BMM_Additive
    rotationSpace=BCS_BoneSpace
    scaleMode=BMM_Additive
    scaleSpace=BCS_BoneSpace
    translationMode=BMM_Additive
    translationSpace=BCS_BoneSpace
    alphaBoolBlend {blendOption=Linear}
    boneToModify {boneName=dyn_coat_r_MSTR}
    rotation {roll=-10, yaw=90}
    translation=<-6 -2 4.18>
   }
   n44 = "Transform (Modify) Bone - Bone: dyn_coat_bk_r_MSTR"!(ComponentPose=@n43, Alpha:float=@n42, Translation="-6,0,0", Rotation="0,0,0", Scale:Vector="1,1,1") -> (Pose) {
    rotationMode=BMM_Additive
    rotationSpace=BCS_BoneSpace
    scaleMode=BMM_Additive
    scaleSpace=BCS_BoneSpace
    translationMode=BMM_Additive
    translationSpace=BCS_BoneSpace
    alphaBoolBlend {blendOption=Linear}
    boneToModify {boneName=dyn_coat_bk_r_MSTR}
    translation {x=-6}
   }
   n45 = "Transform (Modify) Bone - Bone: dyn_coat_r_MSTR"!(
    ComponentPose=@n44
    bAlphaBoolEnabled:bool=Get inCrouchingNoMoving() -> (inCrouchingNoMoving:bool)
    Translation="0,2,2"
    Rotation="0,0,0"
    Scale="1,1,1"
   ) -> (Pose) {
    alphaInputType=Bool
    rotationMode=BMM_Additive
    rotationSpace=BCS_BoneSpace
    scaleMode=BMM_Additive
    scaleSpace=BCS_BoneSpace
    translationMode=BMM_Additive
    translationSpace=BCS_BoneSpace
    alphaBoolBlend {blendOption=Linear}
    boneToModify {boneName=dyn_coat_r_MSTR}
    translation {y=2, z=2}
   }
   n46 = RigidBody!(ComponentPose=@n45, Alpha, OverrideWorldGravity:Vector=Get GravityOverride(), ExternalForce) -> (Pose) {
    bClampLinearTranslationLimitToRefPose=true
    bOverrideWorldGravity=true
    overridePhysicsAsset=&/F_MED_Ultralight_Dyn_skirt
    simulationSpace=BaseBoneSpace
    alphaBoolBlend {blendOption=Linear}
    baseBoneRef {boneName=pelvis}
    componentAppliedLinearAccClamp=<3000 3000 3000>
    componentLinearAccScale=<0.4 0.1 0.1>
    componentLinearVelScale=<0.4 0.1 0.1>
    overrideWorldGravity {z=-980}
    simSpaceSettings {
     maxAngularAcceleration=500
     maxAngularVelocity=3
     maxLinearAcceleration=700
     maxLinearVelocity=500
     velocityScaleZ=0
     worldAlpha=1
     externalLinearDragV=<0.1 0.1>
    }
   }
   n47 = "Transform (Modify) Bone - Bone: dyn_coat_l_MSTR"(
    ComponentPose=Local To Component(LocalPose=Use cached pose 'StartupCache'())
    Alpha=Get thigh_rot_up_l()
    Translation="-6,-2,7"
    Rotation="-15,-90,-5"
    Scale="1,1,1"
   ) {
    rotationMode=BMM_Additive
    rotationSpace=BCS_BoneSpace
    scaleMode=BMM_Additive
    scaleSpace=BCS_BoneSpace
    translationMode=BMM_Additive
    translationSpace=BCS_BoneSpace
    alphaBoolBlend {blendOption=Linear}
    boneToModify {boneName=dyn_coat_l_MSTR}
    rotation=<rot -15 -90 -5>
    translation=<-6 -2 7>
   }
   n48 = "Transform (Modify) Bone - Bone: dyn_coat_r_MSTR"!(ComponentPose=@n47, Alpha:float=Get thigh_rot_up_r(), Translation="-6,-2,7", Rotation="-15,90,0", Scale="1,1,1") -> (Pose) {
    rotationMode=BMM_Additive
    rotationSpace=BCS_BoneSpace
    scaleMode=BMM_Additive
    scaleSpace=BCS_BoneSpace
    translationMode=BMM_Additive
    translationSpace=BCS_BoneSpace
    alphaBoolBlend {blendOption=Linear}
    boneToModify {boneName=dyn_coat_r_MSTR}
    rotation {pitch=-15, yaw=90}
    translation=<-6 -2 7>
   }
   n49 = "Transform (Modify) Bone - Bone: dyn_coat_l_MSTR"(ComponentPose=@n48, Alpha=1, Translation="0,0,0", Rotation="5,0,0", Scale="1,1,1") {
    rotationMode=BMM_Additive
    rotationSpace=BCS_BoneSpace
    scaleMode=BMM_Additive
    scaleSpace=BCS_BoneSpace
    translationMode=BMM_Additive
    translationSpace=BCS_BoneSpace
    alphaBoolBlend {blendOption=Linear}
    boneToModify {boneName=dyn_coat_l_MSTR}
    rotation {pitch=5}
   }
   n50 = "Transform (Modify) Bone - Bone: dyn_coat_r_MSTR"(ComponentPose=@n49, Alpha=1, Translation="0,2,0.5", Rotation="5,0,0", Scale="1,1,1") {
    rotationMode=BMM_Additive
    rotationSpace=BCS_BoneSpace
    scaleMode=BMM_Additive
    scaleSpace=BCS_BoneSpace
    translationMode=BMM_Additive
    translationSpace=BCS_BoneSpace
    alphaBoolBlend {blendOption=Linear}
    boneToModify {boneName=dyn_coat_r_MSTR}
    rotation {pitch=5}
    translation {y=2, z=0.5}
   }
   n51 = Component To Local(ComponentPose=@n50)
   n52 = Get thigh_rot_up_l()
   n53 = "Transform (Modify) Bone - Bone: dyn_coat_l_MSTR"(
    ComponentPose=Local To Component(LocalPose=Use cached pose 'StartupCache'())
    Alpha=@n52
    Translation="-6,-2,7"
    Rotation="0,-80,10"
    Scale="1,1,1"
   ) {
    rotationMode=BMM_Additive
    rotationSpace=BCS_BoneSpace
    scaleMode=BMM_Additive
    scaleSpace=BCS_BoneSpace
    translationMode=BMM_Additive
    translationSpace=BCS_BoneSpace
    alphaBoolBlend {blendOption=Linear}
    boneToModify {boneName=dyn_coat_l_MSTR}
    rotation {roll=10, yaw=-80}
    translation=<-6 -2 7>
   }
   n54 = "Transform (Modify) Bone - Bone: dyn_coat_bk_l_MSTR"(ComponentPose=@n53, Alpha=@n52, Translation="-6,0,0", Rotation="0,0,0", Scale="1,1,1") {
    rotationMode=BMM_Additive
    rotationSpace=BCS_BoneSpace
    scaleMode=BMM_Additive
    scaleSpace=BCS_BoneSpace
    translationMode=BMM_Additive
    translationSpace=BCS_BoneSpace
    alphaBoolBlend {blendOption=Linear}
    boneToModify {boneName=dyn_coat_bk_l_MSTR}
    translation {x=-6}
   }
   n55 = Get thigh_rot_up_r()
   n56 = "Transform (Modify) Bone - Bone: dyn_coat_r_MSTR"(ComponentPose=@n54, Alpha=@n55, Translation="-6,-2,4.180", Rotation="0,90,-10", Scale="1,1,1") {
    rotationMode=BMM_Additive
    rotationSpace=BCS_BoneSpace
    scaleMode=BMM_Additive
    scaleSpace=BCS_BoneSpace
    translationMode=BMM_Additive
    translationSpace=BCS_BoneSpace
    alphaBoolBlend {blendOption=Linear}
    boneToModify {boneName=dyn_coat_r_MSTR}
    rotation {roll=-10, yaw=90}
    translation=<-6 -2 4.18>
   }
   n57 = "Transform (Modify) Bone - Bone: dyn_coat_bk_r_MSTR"(ComponentPose=@n56, Alpha=@n55, Translation="-6,0,0", Rotation="0,0,0", Scale="1,1,1") {
    rotationMode=BMM_Additive
    rotationSpace=BCS_BoneSpace
    scaleMode=BMM_Additive
    scaleSpace=BCS_BoneSpace
    translationMode=BMM_Additive
    translationSpace=BCS_BoneSpace
    alphaBoolBlend {blendOption=Linear}
    boneToModify {boneName=dyn_coat_bk_r_MSTR}
    translation {x=-6}
   }
   n58 = "Transform (Modify) Bone - Bone: dyn_coat_r_MSTR"!(ComponentPose=@n57, bAlphaBoolEnabled:bool=Get inCrouchingNoMoving(), Translation="0,2,2", Rotation="0,0,0", Scale="1,1,1") -> (Pose) {
    alphaInputType=Bool
    rotationMode=BMM_Additive
    rotationSpace=BCS_BoneSpace
    scaleMode=BMM_Additive
    scaleSpace=BCS_BoneSpace
    translationMode=BMM_Additive
    translationSpace=BCS_BoneSpace
    alphaBoolBlend {blendOption=Linear}
    boneToModify {boneName=dyn_coat_r_MSTR}
    translation {y=2, z=2}
   }
   n59 = "Transform (Modify) Bone - Bone: dyn_coat_r_MSTR"!(ComponentPose=@n58, Alpha:float=1, Translation="0,4,0", Rotation="0,0,20", Scale="1,1,1") -> (Pose) {
    rotationMode=BMM_Additive
    rotationSpace=BCS_BoneSpace
    scaleMode=BMM_Additive
    scaleSpace=BCS_BoneSpace
    translationMode=BMM_Additive
    translationSpace=BCS_BoneSpace
    alphaBoolBlend {blendOption=Linear}
    boneToModify {boneName=dyn_coat_r_MSTR}
    rotation {roll=20}
    translation {y=4}
   }
   n60 = RigidBody(ComponentPose=@n59, OverrideWorldGravity=Get GravityOverride()) {
    bClampLinearTranslationLimitToRefPose=true
    bOverrideWorldGravity=true
    overridePhysicsAsset=&F_MED_Ultralight_Dyn_skirt
    simulationSpace=BaseBoneSpace
    alphaBoolBlend {blendOption=Linear}
    baseBoneRef {boneName=pelvis}
    componentAppliedLinearAccClamp=<3000 3000 3000>
    componentLinearAccScale=<0.4 0.1 0.1>
    componentLinearVelScale=<0.4 0.1 0.1>
    overrideWorldGravity {z=-980}
    simSpaceSettings {
     maxAngularAcceleration=500
     maxAngularVelocity=3
     maxLinearAcceleration=700
     maxLinearVelocity=500
     velocityScaleZ=0
     worldAlpha=1
     externalLinearDragV=<0.1 0.1>
    }
   }
   n61 = Component To Local(ComponentPose=@n60)
   n62 = Get bClothEnabled()
   Save cached pose 'CapeCache'(Pose=Blend Poses by bool(
     BlendPose_0=Component To Local(ComponentPose=@n46)
     BlendPose_1=Blend Poses by bool(BlendPose_0=@n51, BlendPose_1=@n61, BlendTime_0=0.2, BlendTime_1=0.5, bActiveValue=@n62) {blendType=Linear}
     BlendTime_0=0.2
     BlendTime_1=0.5
     bActiveValue=Get Is Backbling A Cape() -> (Is Backbling A Cape:bool)
    ) {blendType=Linear}) {cacheName=CapeCache}
   n63 = Get IsCrouchingIdle() -> (IsCrouchingIdle:double)
   n64 = "Transform (Modify) Bone - Bone: dyn_coat_bk_l_02"(ComponentPose=Local To Component(LocalPose=Use cached pose 'CapeCache'()), Alpha=@n63, Translation="0,0,0", Rotation="0,0,57", Scale="1,1,1") {
    rotationMode=BMM_Additive
    scaleMode=BMM_Additive
    translationMode=BMM_Additive
    alphaBoolBlend {blendOption=Linear}
    alphaScaleBiasClamp {bInterpResult=true, interpSpeedDecreasing=8, interpSpeedIncreasing=12}
    boneToModify {boneName=dyn_coat_bk_l_02}
    rotation {roll=57}
   }
   n65 = "Transform (Modify) Bone - Bone: dyn_coat_bk_02"(ComponentPose=@n64, Alpha=@n63, Translation="0,0,0", Rotation="0,0,57", Scale="1,1,1") {
    rotationMode=BMM_Additive
    scaleMode=BMM_Additive
    translationMode=BMM_Additive
    alphaBoolBlend {blendOption=Linear}
    alphaScaleBiasClamp {bInterpResult=true, interpSpeedDecreasing=8, interpSpeedIncreasing=12}
    boneToModify {boneName=dyn_coat_bk_02}
    rotation {roll=57}
   }
   n66 = "Transform (Modify) Bone - Bone: dyn_coat_bk_r_02"(ComponentPose=@n65, Alpha=@n63, Translation="0,0,0", Rotation="0,0,57", Scale="1,1,1") {
    rotationMode=BMM_Additive
    scaleMode=BMM_Additive
    translationMode=BMM_Additive
    alphaBoolBlend {blendOption=Linear}
    alphaScaleBiasClamp {bInterpResult=true, interpSpeedDecreasing=8, interpSpeedIncreasing=12}
    boneToModify {boneName=dyn_coat_bk_r_02}
    rotation {roll=57}
   }
   n67 = "Transform (Modify) Bone - Bone: dyn_coat_r_02"(ComponentPose=@n66, Alpha=@n63, Translation="0,0,0", Rotation="0,0,70", Scale="1,1,1") {
    rotationMode=BMM_Additive
    scaleMode=BMM_Additive
    translationMode=BMM_Additive
    alphaBoolBlend {blendOption=Linear}
    alphaScaleBiasClamp {bInterpResult=true, interpSpeedDecreasing=8, interpSpeedIncreasing=12}
    boneToModify {boneName=dyn_coat_r_02}
    rotation {roll=70}
   }
   n68 = Get IsCrouchingMoving() -> (IsCrouchingMoving:double)
   n69 = "Transform (Modify) Bone - Bone: dyn_coat_bk_l_03"(ComponentPose=@n67, Alpha=@n68, Translation="0,0,0", Rotation="0,0,35", Scale="1,1,1") {
    rotationMode=BMM_Additive
    scaleMode=BMM_Additive
    translationMode=BMM_Additive
    alphaBoolBlend {blendOption=Linear}
    alphaScaleBiasClamp {bInterpResult=true, interpSpeedDecreasing=9, interpSpeedIncreasing=9}
    boneToModify {boneName=dyn_coat_bk_l_03}
    rotation {roll=35}
   }
   n70 = "Transform (Modify) Bone - Bone: dyn_coat_bk_03"(ComponentPose=@n69, Alpha=@n68, Translation="0,0,0", Rotation="0,0,35", Scale="1,1,1") {
    rotationMode=BMM_Additive
    scaleMode=BMM_Additive
    translationMode=BMM_Additive
    alphaBoolBlend {blendOption=Linear}
    alphaScaleBiasClamp {bInterpResult=true, interpSpeedDecreasing=9, interpSpeedIncreasing=9}
    boneToModify {boneName=dyn_coat_bk_03}
    rotation {roll=35}
   }
   n71 = "Transform (Modify) Bone - Bone: dyn_coat_bk_r_03"(ComponentPose=@n70, Alpha=@n68, Translation="0,0,0", Rotation="0,0,35", Scale="1,1,1") {
    rotationMode=BMM_Additive
    scaleMode=BMM_Additive
    translationMode=BMM_Additive
    alphaBoolBlend {blendOption=Linear}
    alphaScaleBiasClamp {bInterpResult=true, interpSpeedDecreasing=9, interpSpeedIncreasing=9}
    boneToModify {boneName=dyn_coat_bk_r_03}
    rotation {roll=35}
   }
   n72 = "Transform (Modify) Bone - Bone: dyn_coat_r_02"(ComponentPose=@n71, Alpha=@n68, Translation="0,0,0", Rotation="0,0,35", Scale="1,1,1") {
    rotationMode=BMM_Additive
    scaleMode=BMM_Additive
    translationMode=BMM_Additive
    alphaBoolBlend {blendOption=Linear}
    alphaScaleBiasClamp {bInterpResult=true, interpSpeedDecreasing=9, interpSpeedIncreasing=9}
    boneToModify {boneName=dyn_coat_r_02}
    rotation {roll=35}
   }
   n73 = "Transform (Modify) Bone - Bone: dyn_coat_l_02"(ComponentPose=@n72, Alpha=@n68, Translation="0,0,0", Rotation="42,-30,20", Scale="1,1,1") {
    rotationMode=BMM_Additive
    scaleMode=BMM_Additive
    translationMode=BMM_Additive
    alphaBoolBlend {blendOption=Linear}
    alphaScaleBiasClamp {bInterpResult=true, interpSpeedDecreasing=9, interpSpeedIncreasing=9}
    boneToModify {boneName=dyn_coat_l_02}
    rotation=<rot 42 -30 20>
   }
   n74 = Get AlphaIs_CrouchMelee() -> (AlphaIs_CrouchMelee:double)
   n75 = "Transform (Modify) Bone - Bone: dyn_coat_bk_l_MSTR"(ComponentPose=@n73, Alpha=@n74, Translation="0,0,0", Rotation="0,0,15", Scale="1,1,1") {
    rotationMode=BMM_Additive
    scaleMode=BMM_Additive
    translationMode=BMM_Additive
    alphaBoolBlend {blendOption=Linear}
    alphaScaleBiasClamp {bInterpResult=true, interpSpeedDecreasing=9, interpSpeedIncreasing=9}
    boneToModify {boneName=dyn_coat_bk_l_MSTR}
    rotation {roll=15}
   }
   n76 = "Transform (Modify) Bone - Bone: dyn_coat_bk_01"(ComponentPose=@n75, Alpha=@n74, Translation="0,0,0", Rotation="0,0,15", Scale="1,1,1") {
    rotationMode=BMM_Additive
    scaleMode=BMM_Additive
    translationMode=BMM_Additive
    alphaBoolBlend {blendOption=Linear}
    alphaScaleBiasClamp {bInterpResult=true, interpSpeedDecreasing=3, interpSpeedIncreasing=3}
    boneToModify {boneName=dyn_coat_bk_01}
    rotation {roll=15}
   }
   n77 = "Transform (Modify) Bone - Bone: dyn_coat_bk_r_01"(ComponentPose=@n76, Alpha=@n74, Translation="0,0,0", Rotation="0,0,15", Scale="1,1,1") {
    rotationMode=BMM_Additive
    scaleMode=BMM_Additive
    translationMode=BMM_Additive
    alphaBoolBlend {blendOption=Linear}
    alphaScaleBiasClamp {bInterpResult=true, interpSpeedDecreasing=3, interpSpeedIncreasing=3}
    boneToModify {boneName=dyn_coat_bk_r_01}
    rotation {roll=15}
   }
   n78 = "Transform (Modify) Bone - Bone: dyn_coat_r_01"(ComponentPose=@n77, Alpha=@n74, Translation="0,0,0", Rotation="-25,0,0", Scale="1,1,1") {
    rotationMode=BMM_Additive
    scaleMode=BMM_Additive
    translationMode=BMM_Additive
    alphaBoolBlend {blendOption=Linear}
    alphaScaleBiasClamp {bInterpResult=true, interpSpeedDecreasing=3, interpSpeedIncreasing=3}
    boneToModify {boneName=dyn_coat_r_01}
    rotation {pitch=-25}
   }
   n79 = "Transform (Modify) Bone - Bone: dyn_coat_bk_r_03"(
    ComponentPose=@n78
    Alpha=Get IsCrouchingWithWeapons() -> (IsCrouchingWithWeapons:double)
    Translation="0,0,11"
    Rotation="0,0,-4"
    Scale="1,1,1"
   ) {
    rotationMode=BMM_Additive
    scaleMode=BMM_Additive
    translationMode=BMM_Additive
    alphaBoolBlend {blendOption=Linear}
    alphaScaleBiasClamp {bInterpResult=true, interpSpeedDecreasing=9, interpSpeedIncreasing=9}
    boneToModify {boneName=dyn_coat_bk_r_03}
    rotation {roll=-4}
    translation {z=11}
   }
   n80 = Get IsSliding() -> (IsSliding:double)
   n81 = "Transform (Modify) Bone - Bone: dyn_coat_bk_l_MSTR"(ComponentPose=@n79, Alpha=@n80, Translation="0,0,0", Rotation="0,0,70", Scale="1,1,1") {
    rotationMode=BMM_Additive
    scaleMode=BMM_Additive
    translationMode=BMM_Additive
    alphaBoolBlend {blendOption=Linear}
    alphaScaleBiasClamp {bInterpResult=true, interpSpeedDecreasing=9, interpSpeedIncreasing=9}
    boneToModify {boneName=dyn_coat_bk_l_MSTR}
    rotation {roll=70}
   }
   n82 = "Transform (Modify) Bone - Bone: dyn_coat_bk_l_04"(ComponentPose=@n81, Alpha=@n80, Translation="0,0,20", Rotation="0,0,-70", Scale="1,1,1") {
    rotationMode=BMM_Additive
    scaleMode=BMM_Additive
    translationMode=BMM_Additive
    alphaBoolBlend {blendOption=Linear}
    alphaScaleBiasClamp {bInterpResult=true, interpSpeedDecreasing=9, interpSpeedIncreasing=9}
    boneToModify {boneName=dyn_coat_bk_l_04}
    rotation {roll=-70}
    translation {z=20}
   }
   n83 = "Transform (Modify) Bone - Bone: dyn_coat_l_01"(ComponentPose=@n82, Alpha=@n80, Translation="0,-8,5", Rotation="0,0,0", Scale="1,1,1") {
    rotationMode=BMM_Additive
    scaleMode=BMM_Additive
    translationMode=BMM_Additive
    alphaBoolBlend {blendOption=Linear}
    alphaScaleBiasClamp {bInterpResult=true, interpSpeedDecreasing=9, interpSpeedIncreasing=9}
    boneToModify {boneName=dyn_coat_l_01}
    translation {y=-8, z=5}
   }
   n84 = "Transform (Modify) Bone - Bone: dyn_coat_bk_01"(ComponentPose=@n83, Alpha=@n80, Translation="0,0,5", Rotation="0,0,30", Scale="1,1,1") {
    rotationMode=BMM_Additive
    scaleMode=BMM_Additive
    translationMode=BMM_Additive
    alphaBoolBlend {blendOption=Linear}
    alphaScaleBiasClamp {bInterpResult=true, interpSpeedDecreasing=9, interpSpeedIncreasing=9}
    boneToModify {boneName=dyn_coat_bk_01}
    rotation {roll=30}
    translation {z=5}
   }
   n85 = "Transform (Modify) Bone - Bone: dyn_coat_r_MSTR"(ComponentPose=@n84, Alpha=@n80, Translation="-5,-15,0", Rotation="0,0,-20", Scale="0.8,1,1") {
    rotationMode=BMM_Additive
    scaleMode=BMM_Additive
    translationMode=BMM_Additive
    alphaBoolBlend {blendOption=Linear}
    alphaScaleBiasClamp {bInterpResult=true, interpSpeedDecreasing=9, interpSpeedIncreasing=9}
    boneToModify {boneName=dyn_coat_r_MSTR}
    rotation {roll=-20}
    scale {x=0.8}
    translation=<-5 -15>
   }
   n86 = "Transform (Modify) Bone - Bone: dyn_coat_r_02"(ComponentPose=@n85, Alpha=@n80, Translation="3,0,0", Rotation="0,0,0", Scale="0.7,1,1") {
    rotationMode=BMM_Additive
    scaleMode=BMM_Additive
    translationMode=BMM_Additive
    alphaBoolBlend {blendOption=Linear}
    alphaScaleBiasClamp {bInterpResult=true, interpSpeedDecreasing=9, interpSpeedIncreasing=9}
    boneToModify {boneName=dyn_coat_r_02}
    scale {x=0.7}
    translation {x=3}
   }
   n87 = "Transform (Modify) Bone - Bone: dyn_coat_bk_05"(ComponentPose=@n86, Alpha=@n80, Translation="0,20,0", Rotation="0,0,-90", Scale="1,1,1") {
    rotationMode=BMM_Additive
    scaleMode=BMM_Additive
    translationMode=BMM_Additive
    alphaBoolBlend {blendOption=Linear}
    alphaScaleBiasClamp {bInterpResult=true, interpSpeedDecreasing=9, interpSpeedIncreasing=9}
    boneToModify {boneName=dyn_coat_bk_05}
    rotation {roll=-90}
    translation {y=20}
   }
   n88 = "Transform (Modify) Bone - Bone: dyn_coat_bk_r_05"(ComponentPose=@n87, Alpha=@n80, Translation="0,0,0", Rotation="0,0,-50", Scale="1,1,1") {
    rotationMode=BMM_Additive
    scaleMode=BMM_Additive
    translationMode=BMM_Additive
    alphaBoolBlend {blendOption=Linear}
    alphaScaleBiasClamp {bInterpResult=true, interpSpeedDecreasing=9, interpSpeedIncreasing=9}
    boneToModify {boneName=dyn_coat_bk_r_05}
    rotation {roll=-50}
   }
   n89 = "Transform (Modify) Bone - Bone: dyn_coat_bk_r_03"(ComponentPose=@n88, Alpha=@n80, Translation="0,0,20", Rotation="0,0,0", Scale="1,1,1") {
    rotationMode=BMM_Additive
    scaleMode=BMM_Additive
    translationMode=BMM_Additive
    alphaBoolBlend {blendOption=Linear}
    alphaScaleBiasClamp {bInterpResult=true, interpSpeedDecreasing=9, interpSpeedIncreasing=9}
    boneToModify {boneName=dyn_coat_bk_r_03}
    translation {z=20}
   }
   n90 = "Transform (Modify) Bone - Bone: dyn_coat_bk_l_01"(ComponentPose=@n89, Alpha=@n80, Translation="0,0,5", Rotation="0,20,0", Scale="1,1,1") {
    rotationMode=BMM_Additive
    scaleMode=BMM_Additive
    translationMode=BMM_Additive
    alphaBoolBlend {blendOption=Linear}
    alphaScaleBiasClamp {bInterpResult=true, interpSpeedDecreasing=9, interpSpeedIncreasing=9}
    boneToModify {boneName=dyn_coat_bk_l_01}
    rotation {yaw=20}
    translation {z=5}
   }
   n91 = "Transform (Modify) Bone - Bone: dyn_coat_bk_03"(ComponentPose=@n90, Alpha=@n80, Translation="0,0,15", Rotation="0,0,0", Scale="1,1,1") {
    rotationMode=BMM_Additive
    scaleMode=BMM_Additive
    translationMode=BMM_Additive
    alphaBoolBlend {blendOption=Linear}
    alphaScaleBiasClamp {bInterpResult=true, interpSpeedDecreasing=9, interpSpeedIncreasing=9}
    boneToModify {boneName=dyn_coat_bk_03}
    translation {z=15}
   }
   Save cached pose 'TransformsCrouchCache'(Pose=Component To Local(ComponentPose=@n91)) {cacheName=TransformsCrouchCache}
   n92 = Look At - Bone: dyn_coat_bk_r_03(
    ComponentPose=Local To Component(LocalPose=Use cached pose 'TransformsCrouchCache'())
    Alpha=Get IsLegRotationRight() -> (IsLegRotationRight:double)
   ) {
    alphaBoolBlend {blendOption=Linear}
    boneToModify {boneName=dyn_coat_bk_r_03}
    lookAtLocation=<5.7271 4.7541 6.9882>
    lookAtTarget {boneReference {boneName=foot_r}}
    lookAt_Axis {axis=<1 0>}
   }
   n93 = Look At - Bone: dyn_coat_bk_l_03(ComponentPose=@n92, Alpha=Get IsLegRotationLeft() -> (IsLegRotationLeft:double)) {
    alphaBoolBlend {blendOption=Linear}
    boneToModify {boneName=dyn_coat_bk_l_03}
    lookAtLocation=<-15.1584 -3.1952 -3.5367>
    lookAtTarget {boneReference {boneName=foot_l}}
    lookAt_Axis {axis=<1 0>}
   }
   n94 = "Constraint - dyn_coat_bk_03 (RR)"!(ComponentPose=@n93, Alpha=1, ConstraintWeights_0=1, ConstraintWeights_1:float=1) -> (Pose) {
    alphaBoolBlend {blendOption=Linear}
    boneToModify {boneName=dyn_coat_bk_03}
    constraintSetup [
     {offsetOption=None, transformType=Rotation, targetBone {boneName=dyn_coat_bk_r_03}}
     {offsetOption=None, transformType=Rotation, targetBone {boneName=dyn_coat_bk_l_03}}
    ]
    constraintWeights [1,1]
   }
   n95 = Get IsStandMoving() -> (IsStandMoving:double)
   n96 = "Transform (Modify) Bone - Bone: dyn_coat_bk_MSTR"!(ComponentPose=@n94, Alpha:float=@n95, Translation="0,0,0", Rotation="0,0,45", Scale:Vector="1,1,1") -> (Pose) {
    rotationMode=BMM_Additive
    scaleMode=BMM_Additive
    translationMode=BMM_Additive
    alphaBoolBlend {blendOption=Linear}
    alphaScaleBiasClamp {bInterpResult=true, interpSpeedDecreasing=5, interpSpeedIncreasing=9}
    boneToModify {boneName=dyn_coat_bk_MSTR}
    rotation {roll=45}
   }
   n97 = "Transform (Modify) Bone - Bone: dyn_coat_bk_l_MSTR"(ComponentPose=@n96, Alpha=@n95, Translation="0,0,0", Rotation="7,0,25", Scale="1,1,1") {
    rotationMode=BMM_Additive
    scaleMode=BMM_Additive
    translationMode=BMM_Additive
    alphaBoolBlend {blendOption=Linear}
    alphaScaleBiasClamp {bInterpResult=true, interpSpeedDecreasing=5, interpSpeedIncreasing=9}
    boneToModify {boneName=dyn_coat_bk_l_MSTR}
    rotation {pitch=7, roll=25}
   }
   n98 = "Transform (Modify) Bone - Bone: dyn_coat_bk_r_MSTR"(ComponentPose=@n97, Alpha=@n95, Translation="0,0,0", Rotation="-20,-10,25", Scale="1,1,1") {
    rotationMode=BMM_Additive
    scaleMode=BMM_Additive
    translationMode=BMM_Additive
    alphaBoolBlend {blendOption=Linear}
    alphaScaleBiasClamp {bInterpResult=true, interpSpeedDecreasing=5, interpSpeedIncreasing=9}
    boneToModify {boneName=dyn_coat_bk_r_MSTR}
    rotation=<rot -20 -10 25>
   }
   n99 = "Transform (Modify) Bone - Bone: dyn_coat_r_MSTR"(ComponentPose=@n98, Alpha=@n95, Translation="-3,-3,0", Rotation="-20,0,0", Scale="1,1,1") {
    rotationMode=BMM_Additive
    scaleMode=BMM_Additive
    translationMode=BMM_Additive
    alphaBoolBlend {blendOption=Linear}
    alphaScaleBiasClamp {bInterpResult=true, interpSpeedDecreasing=5, interpSpeedIncreasing=9}
    boneToModify {boneName=dyn_coat_r_MSTR}
    rotation {pitch=-20}
    translation=<-3 -3>
   }
   n100 = "Transform (Modify) Bone - Bone: dyn_coat_l_01"(ComponentPose=@n99, Alpha=@n95, Translation="0,0,0", Rotation="20,0,0", Scale="1,1,1") {
    rotationMode=BMM_Additive
    scaleMode=BMM_Additive
    translationMode=BMM_Additive
    alphaBoolBlend {blendOption=Linear}
    alphaScaleBiasClamp {bInterpResult=true, interpSpeedDecreasing=5, interpSpeedIncreasing=9}
    boneToModify {boneName=dyn_coat_l_01}
    rotation {pitch=20}
   }
   n101 = "Transform (Modify) Bone - Bone: dyn_coat_bk_l_03"(ComponentPose=@n100, Alpha=@n95, Translation="0,0,0", Rotation="25,0,0", Scale="1,1,1") {
    rotationMode=BMM_Additive
    scaleMode=BMM_Additive
    translationMode=BMM_Additive
    alphaBoolBlend {blendOption=Linear}
    alphaScaleBiasClamp {bInterpResult=true, interpSpeedDecreasing=5, interpSpeedIncreasing=9}
    boneToModify {boneName=dyn_coat_bk_l_03}
    rotation {pitch=25}
   }
   n102 = Get Alpha_isStandingIddle() -> (Alpha_isStandingIddle:float)
   n103 = "Transform (Modify) Bone - Bone: dyn_coat_bk_r_MSTR"(ComponentPose=@n101, Alpha=@n102, Translation="0,0,0", Rotation="-3,0,10", Scale="1,1,1") {
    rotationMode=BMM_Additive
    scaleMode=BMM_Additive
    translationMode=BMM_Additive
    alphaBoolBlend {blendOption=Linear}
    alphaScaleBiasClamp {bInterpResult=true, interpSpeedDecreasing=3, interpSpeedIncreasing=5}
    boneToModify {boneName=dyn_coat_bk_r_MSTR}
    rotation {pitch=-3, roll=10}
   }
   n104 = "Transform (Modify) Bone - Bone: dyn_coat_r_MSTR"(ComponentPose=@n103, Alpha=@n102, Translation="0,0,0", Rotation="-5,0,0", Scale="1,1,1") {
    rotationMode=BMM_Additive
    scaleMode=BMM_Additive
    translationMode=BMM_Additive
    alphaBoolBlend {blendOption=Linear}
    alphaScaleBiasClamp {bInterpResult=true, interpSpeedDecreasing=3, interpSpeedIncreasing=5}
    boneToModify {boneName=dyn_coat_r_MSTR}
    rotation {pitch=-5}
   }
   Save cached pose 'SkeletalControlsCache'(Pose=Component To Local(ComponentPose=@n104)) {cacheName=SkeletalControlsCache}
   n105 = Use cached pose 'TransformsCrouchCache'()
   n106 = Use cached pose 'SkeletalControlsCache'()
   n107 = Get IsCrouchingDetected() -> (IsCrouchingDetected:bool)
   n108 = "Transform (Modify) Bone - Bone: dyn_coat_bk_l_01"(
    ComponentPose=Local To Component(LocalPose=Blend Poses by bool(BlendPose_0=@n105, BlendPose_1=@n106, BlendTime_0=0.2, BlendTime_1=0.5, bActiveValue=@n107) {blendType=Linear})
    Alpha=@n1
    Translation="0,0,5"
    Rotation="0,10,0"
    Scale="1,1,1"
   ) {
    rotationMode=BMM_Additive
    scaleMode=BMM_Additive
    translationMode=BMM_Additive
    alphaBoolBlend {blendOption=Linear}
    alphaScaleBiasClamp {bInterpResult=true, interpSpeedDecreasing=9, interpSpeedIncreasing=9}
    boneToModify {boneName=dyn_coat_bk_l_01}
    rotation {yaw=10}
    translation {z=5}
   }
   n109 = "Transform (Modify) Bone - Bone: dyn_coat_bk_l_02"(ComponentPose=@n108, Alpha=@n1, Translation="0,0,4", Rotation="0,0,0", Scale="1,1,1") {
    rotationMode=BMM_Additive
    scaleMode=BMM_Additive
    translationMode=BMM_Additive
    alphaBoolBlend {blendOption=Linear}
    alphaScaleBiasClamp {bInterpResult=true, interpSpeedDecreasing=9, interpSpeedIncreasing=9}
    boneToModify {boneName=dyn_coat_bk_l_02}
    translation {z=4}
   }
   n110 = "Transform (Modify) Bone - Bone: dyn_coat_bk_r_03"(ComponentPose=@n109, Alpha=@n1, Translation="0,0,5", Rotation="10,0,0", Scale="1,1,1") {
    rotationMode=BMM_Additive
    scaleMode=BMM_Additive
    translationMode=BMM_Additive
    alphaBoolBlend {blendOption=Linear}
    alphaScaleBiasClamp {bInterpResult=true, interpSpeedDecreasing=9, interpSpeedIncreasing=9}
    boneToModify {boneName=dyn_coat_bk_r_03}
    rotation {pitch=10}
    translation {z=5}
   }
   n111 = Get Alpha_isSkidiveUp() -> (Alpha_isSkidiveUp:float)
   n112 = "Transform (Modify) Bone - Bone: dyn_coat_bk_l_01"(ComponentPose=@n110, Alpha=@n111, Translation="0,0,2", Rotation="0,0,10", Scale="1,1,1") {
    rotationMode=BMM_Additive
    scaleMode=BMM_Additive
    translationMode=BMM_Additive
    alphaBoolBlend {blendOption=Linear}
    alphaScaleBiasClamp {bInterpResult=true, interpSpeedDecreasing=2, interpSpeedIncreasing=9}
    boneToModify {boneName=dyn_coat_bk_l_01}
    rotation {roll=10}
    translation {z=2}
   }
   n113 = "Transform (Modify) Bone - Bone: dyn_coat_bk_01"(ComponentPose=@n112, Alpha=@n111, Translation="0,0,2", Rotation="0,0,10", Scale="1,1,1") {
    rotationMode=BMM_Additive
    scaleMode=BMM_Additive
    translationMode=BMM_Additive
    alphaBoolBlend {blendOption=Linear}
    alphaScaleBiasClamp {bInterpResult=true, interpSpeedDecreasing=2, interpSpeedIncreasing=9}
    boneToModify {boneName=dyn_coat_bk_01}
    rotation {roll=10}
    translation {z=2}
   }
   n114 = "Transform (Modify) Bone - Bone: dyn_coat_bk_r_MSTR"(ComponentPose=@n113, Alpha=@n111, Translation="0,0,6", Rotation="-5,0,15", Scale="1,1,1") {
    rotationMode=BMM_Additive
    scaleMode=BMM_Additive
    translationMode=BMM_Additive
    alphaBoolBlend {blendOption=Linear}
    alphaScaleBiasClamp {bInterpResult=true, interpSpeedDecreasing=2, interpSpeedIncreasing=9}
    boneToModify {boneName=dyn_coat_bk_r_MSTR}
    rotation {pitch=-5, roll=15}
    translation {z=6}
   }
   n115 = Get Is_DBNO() -> (Is_DBNO:float)
   n116 = "Transform (Modify) Bone - Bone: dyn_coat_bk_r_03"(ComponentPose=@n114, Alpha=@n115, Translation="0,0,0", Rotation="0,0,-30", Scale="1,1,1") {
    rotationMode=BMM_Additive
    scaleMode=BMM_Additive
    translationMode=BMM_Additive
    alphaBoolBlend {blendOption=Linear}
    alphaScaleBiasClamp {bInterpResult=true, interpSpeedDecreasing=9, interpSpeedIncreasing=9}
    boneToModify {boneName=dyn_coat_bk_r_03}
    rotation {roll=-30}
   }
   n117 = "Transform (Modify) Bone - Bone: dyn_coat_bk_l_03"(ComponentPose=@n116, Alpha=@n115, Translation="0,0,0", Rotation="0,0,-30", Scale="1,1,1") {
    rotationMode=BMM_Additive
    scaleMode=BMM_Additive
    translationMode=BMM_Additive
    alphaBoolBlend {blendOption=Linear}
    alphaScaleBiasClamp {bInterpResult=true, interpSpeedDecreasing=9, interpSpeedIncreasing=9}
    boneToModify {boneName=dyn_coat_bk_l_03}
    rotation {roll=-30}
   }
   n118 = "Transform (Modify) Bone - Bone: dyn_coat_bk_03"(ComponentPose=@n117, Alpha=@n115, Translation="0,0,0", Rotation="0,0,-30", Scale="1,1,1") {
    rotationMode=BMM_Additive
    scaleMode=BMM_Additive
    translationMode=BMM_Additive
    alphaBoolBlend {blendOption=Linear}
    alphaScaleBiasClamp {bInterpResult=true, interpSpeedDecreasing=9, interpSpeedIncreasing=9}
    boneToModify {boneName=dyn_coat_bk_03}
    rotation {roll=-30}
   }
   n119 = "Transform (Modify) Bone - Bone: dyn_coat_bk_r_04"(ComponentPose=@n118, Alpha=@n115, Translation="0,0,0", Rotation="0,0,30", Scale="1,1,1") {
    rotationMode=BMM_Additive
    scaleMode=BMM_Additive
    translationMode=BMM_Additive
    alphaBoolBlend {blendOption=Linear}
    alphaScaleBiasClamp {bInterpResult=true, interpSpeedDecreasing=9, interpSpeedIncreasing=9}
    boneToModify {boneName=dyn_coat_bk_r_04}
    rotation {roll=30}
   }
   n120 = "Transform (Modify) Bone - Bone: dyn_coat_bk_l_04"(ComponentPose=@n119, Alpha=@n115, Translation="0,0,0", Rotation="0,0,30", Scale="1,1,1") {
    rotationMode=BMM_Additive
    scaleMode=BMM_Additive
    translationMode=BMM_Additive
    alphaBoolBlend {blendOption=Linear}
    alphaScaleBiasClamp {bInterpResult=true, interpSpeedDecreasing=9, interpSpeedIncreasing=9}
    boneToModify {boneName=dyn_coat_bk_l_04}
    rotation {roll=30}
   }
   n121 = "Transform (Modify) Bone - Bone: dyn_coat_bk_04"(ComponentPose=@n120, Alpha=@n115, Translation="0,0,0", Rotation="0,0,30", Scale="1,1,1") {
    rotationMode=BMM_Additive
    scaleMode=BMM_Additive
    translationMode=BMM_Additive
    alphaBoolBlend {blendOption=Linear}
    alphaScaleBiasClamp {bInterpResult=true, interpSpeedDecreasing=9, interpSpeedIncreasing=9}
    boneToModify {boneName=dyn_coat_bk_04}
    rotation {roll=30}
   }
   n122 = Look At - Bone: dyn_coat_l_MSTR(ComponentPose=@n121, Alpha=float * float(A=@n115, B=0.5) {is=title K2Node_PromotableOperator}) {
    alphaBoolBlend {blendOption=Linear}
    boneToModify {boneName=dyn_coat_l_MSTR}
    lookAtLocation=<3.1322 4.7541 -8.1549>
    lookAtTarget {boneReference {boneName=foot_l}}
    lookAt_Axis {axis=<1 0>}
   }
   n123 = Get Alpha_isParachute() -> (Alpha_isParachute:float)
   n124 = "Transform (Modify) Bone - Bone: dyn_coat_r_MSTR"(ComponentPose=@n122, Alpha=@n123, Translation="-2,0,0", Rotation="0,0,0", Scale="1,1,1") {
    rotationMode=BMM_Additive
    scaleMode=BMM_Additive
    translationMode=BMM_Additive
    alphaBoolBlend {blendOption=Linear}
    alphaScaleBiasClamp {bInterpResult=true, interpSpeedDecreasing=9, interpSpeedIncreasing=9}
    boneToModify {boneName=dyn_coat_r_MSTR}
    translation {x=-2}
   }
   n125 = "Transform (Modify) Bone - Bone: dyn_coat_bk_r_01"(ComponentPose=@n124, Alpha=@n123, Translation="-1,0,0", Rotation="-10,0,10", Scale="1,1,1") {
    rotationMode=BMM_Additive
    scaleMode=BMM_Additive
    translationMode=BMM_Additive
    alphaBoolBlend {blendOption=Linear}
    alphaScaleBiasClamp {bInterpResult=true, interpSpeedDecreasing=9, interpSpeedIncreasing=9}
    boneToModify {boneName=dyn_coat_bk_r_01}
    rotation {pitch=-10, roll=10}
    translation {x=-1}
   }
   n126 = "Transform (Modify) Bone - Bone: dyn_coat_l_01"(ComponentPose=@n125, Alpha=@n123, Translation="3,0,0", Rotation="0,0,0", Scale="1,1,1") {
    rotationMode=BMM_Additive
    scaleMode=BMM_Additive
    translationMode=BMM_Additive
    alphaBoolBlend {blendOption=Linear}
    alphaScaleBiasClamp {bInterpResult=true, interpSpeedDecreasing=9, interpSpeedIncreasing=2}
    boneToModify {boneName=dyn_coat_l_01}
    translation {x=3}
   }
   n127 = "Transform (Modify) Bone - Bone: dyn_coat_bk_l_MSTR"(ComponentPose=@n126, Alpha=@n123, Translation="0,0,0", Rotation="15,0,10", Scale="1,1,1") {
    rotationMode=BMM_Additive
    scaleMode=BMM_Additive
    translationMode=BMM_Additive
    alphaBoolBlend {blendOption=Linear}
    alphaScaleBiasClamp {bInterpResult=true, interpSpeedDecreasing=9, interpSpeedIncreasing=9}
    boneToModify {boneName=dyn_coat_bk_l_MSTR}
    rotation {pitch=15, roll=10}
   }
   n128 = "Transform (Modify) Bone - Bone: dyn_coat_bk_01"(ComponentPose=@n127, Alpha=@n123, Translation="0,0,0", Rotation="0,0,20", Scale="1,1,1") {
    rotationMode=BMM_Additive
    scaleMode=BMM_Additive
    translationMode=BMM_Additive
    alphaBoolBlend {blendOption=Linear}
    alphaScaleBiasClamp {bInterpResult=true, interpSpeedDecreasing=9, interpSpeedIncreasing=2}
    boneToModify {boneName=dyn_coat_bk_01}
    rotation {roll=20}
   }
   n129 = "Transform (Modify) Bone - Bone: dyn_coat_bk_l_03"(ComponentPose=@n128, Alpha=@n123, Translation="0,0,0", Rotation="5,0,0", Scale="1,1,1") {
    rotationMode=BMM_Additive
    scaleMode=BMM_Additive
    translationMode=BMM_Additive
    alphaBoolBlend {blendOption=Linear}
    alphaScaleBiasClamp {bInterpResult=true, interpSpeedDecreasing=9, interpSpeedIncreasing=2}
    boneToModify {boneName=dyn_coat_bk_l_03}
    rotation {pitch=5}
   }
   n130 = "Transform (Modify) Bone - Bone: dyn_coat_bk_r_03"(ComponentPose=@n129, Alpha=@n123, Translation="2,0,0", Rotation="0,-20,0", Scale="1,1,1") {
    rotationMode=BMM_Additive
    scaleMode=BMM_Additive
    translationMode=BMM_Additive
    alphaBoolBlend {blendOption=Linear}
    alphaScaleBiasClamp {bInterpResult=true, interpSpeedDecreasing=9, interpSpeedIncreasing=2}
    boneToModify {boneName=dyn_coat_bk_r_03}
    rotation {yaw=-20}
    translation {x=2}
   }
   n131 = Get AlphaIs_SkydiveMode() -> (AlphaIs_SkydiveMode:double)
   n132 = "Transform (Modify) Bone - Bone: dyn_coat_bk_l_01"(ComponentPose=@n130, Alpha=@n131, Translation="2.5,0,0", Rotation="0,0,0", Scale="1,1,1") {
    rotationMode=BMM_Additive
    scaleMode=BMM_Additive
    translationMode=BMM_Additive
    alphaBoolBlend {blendOption=Linear}
    alphaScaleBiasClamp {bInterpResult=true, interpSpeedDecreasing=9, interpSpeedIncreasing=9}
    boneToModify {boneName=dyn_coat_bk_l_01}
    translation {x=2.5}
   }
   n133 = "Transform (Modify) Bone - Bone: dyn_coat_r_01"(ComponentPose=@n132, Alpha=@n131, Translation="0,0,0", Rotation="0,0,0", Scale="1,1,1") {
    rotationMode=BMM_Additive
    scaleMode=BMM_Additive
    translationMode=BMM_Additive
    alphaBoolBlend {blendOption=Linear}
    alphaScaleBiasClamp {bInterpResult=true, interpSpeedDecreasing=9, interpSpeedIncreasing=9}
    boneToModify {boneName=dyn_coat_r_01}
   }
   n134 = Get Is_GrindRail() -> (Is_GrindRail:float)
   n135 = "Transform (Modify) Bone - Bone: dyn_coat_l_01"(ComponentPose=@n133, Alpha=@n134, Translation="0,0,0", Rotation="70,0,0", Scale="1,1,1") {
    rotationMode=BMM_Additive
    scaleMode=BMM_Additive
    translationMode=BMM_Additive
    alphaBoolBlend {blendOption=Linear}
    alphaScaleBiasClamp {bInterpResult=true, interpSpeedDecreasing=9, interpSpeedIncreasing=9}
    boneToModify {boneName=dyn_coat_l_01}
    rotation {pitch=70}
   }
   n136 = "Transform (Modify) Bone - Bone: dyn_coat_bk_l_01"(ComponentPose=@n135, Alpha=@n134, Translation="0,0,0", Rotation="70,0,0", Scale="1,1,1") {
    rotationMode=BMM_Additive
    scaleMode=BMM_Additive
    translationMode=BMM_Additive
    alphaBoolBlend {blendOption=Linear}
    alphaScaleBiasClamp {bInterpResult=true, interpSpeedDecreasing=9, interpSpeedIncreasing=9}
    boneToModify {boneName=dyn_coat_bk_l_01}
    rotation {pitch=70}
   }
   n137 = "Transform (Modify) Bone - Bone: dyn_coat_bk_01"(ComponentPose=@n136, Alpha=@n134, Translation="0,0,0", Rotation="70,0,0", Scale="1,1,1") {
    rotationMode=BMM_Additive
    scaleMode=BMM_Additive
    translationMode=BMM_Additive
    alphaBoolBlend {blendOption=Linear}
    alphaScaleBiasClamp {bInterpResult=true, interpSpeedDecreasing=9, interpSpeedIncreasing=9}
    boneToModify {boneName=dyn_coat_bk_01}
    rotation {pitch=70}
   }
   n138 = "Transform (Modify) Bone - Bone: dyn_coat_bk_r_01"(ComponentPose=@n137, Alpha=@n134, Translation="0,0,0", Rotation="-20,50,20", Scale="1,1,1") {
    rotationMode=BMM_Additive
    scaleMode=BMM_Additive
    translationMode=BMM_Additive
    alphaBoolBlend {blendOption=Linear}
    alphaScaleBiasClamp {bInterpResult=true, interpSpeedDecreasing=9, interpSpeedIncreasing=9}
    boneToModify {boneName=dyn_coat_bk_r_01}
    rotation=<rot -20 50 20>
   }
   n139 = "Transform (Modify) Bone - Bone: dyn_coat_r_01"(ComponentPose=@n138, Alpha=@n134, Translation="3,0,0", Rotation="-10,20,52", Scale="1,1,1") {
    rotationMode=BMM_Additive
    scaleMode=BMM_Additive
    translationMode=BMM_Additive
    alphaBoolBlend {blendOption=Linear}
    alphaScaleBiasClamp {bInterpResult=true, interpSpeedDecreasing=9, interpSpeedIncreasing=9}
    boneToModify {boneName=dyn_coat_r_01}
    rotation=<rot -10 20 52>
    translation {x=3}
   }
   n140 = "Transform (Modify) Bone - Bone: dyn_coat_bk_r_03"(ComponentPose=@n139, Alpha=@n134, Translation="0,0,0", Rotation="0,-50,0", Scale="1,1,1") {
    rotationMode=BMM_Additive
    scaleMode=BMM_Additive
    translationMode=BMM_Additive
    alphaBoolBlend {blendOption=Linear}
    alphaScaleBiasClamp {bInterpResult=true, interpSpeedDecreasing=9, interpSpeedIncreasing=9}
    boneToModify {boneName=dyn_coat_bk_r_03}
    rotation {yaw=-50}
   }
   n141 = Get Alpha_isRotateLegR_OnRail() -> (Alpha_isRotateLegR_OnRail:float)
   n142 = "Transform (Modify) Bone - Bone: dyn_coat_bk_r_01"(ComponentPose=@n140, Alpha=@n141, Translation="4.5,-6,0", Rotation="0,10,10", Scale="1,1,1") {
    rotationMode=BMM_Additive
    scaleMode=BMM_Additive
    translationMode=BMM_Additive
    alphaBoolBlend {blendOption=Linear}
    alphaScaleBiasClamp {bInterpResult=true, interpSpeedDecreasing=9, interpSpeedIncreasing=9}
    boneToModify {boneName=dyn_coat_bk_r_01}
    rotation {roll=10, yaw=10}
    translation=<4.5 -6>
   }
   n143 = "Transform (Modify) Bone - Bone: dyn_coat_l_02"(ComponentPose=@n142, Alpha=@n141, Translation="0,0,0", Rotation="42,0,20", Scale="1,1,1") {
    rotationMode=BMM_Additive
    scaleMode=BMM_Additive
    translationMode=BMM_Additive
    alphaBoolBlend {blendOption=Linear}
    alphaScaleBiasClamp {bInterpResult=true, interpSpeedDecreasing=9, interpSpeedIncreasing=9}
    boneToModify {boneName=dyn_coat_l_02}
    rotation {pitch=42, roll=20}
   }
   n144 = "Transform (Modify) Bone - Bone: dyn_coat_bk_r_02"(ComponentPose=@n143, Alpha=@n141, Translation="0,-2,-5", Rotation="30,0,0", Scale="1,1,1") {
    rotationMode=BMM_Additive
    scaleMode=BMM_Additive
    translationMode=BMM_Additive
    alphaBoolBlend {blendOption=Linear}
    alphaScaleBiasClamp {bInterpResult=true, interpSpeedDecreasing=9, interpSpeedIncreasing=9}
    boneToModify {boneName=dyn_coat_bk_r_02}
    rotation {pitch=30}
    translation {y=-2, z=-5}
   }
   n145 = "Transform (Modify) Bone - Bone: dyn_coat_r_01"(ComponentPose=@n144, Alpha=@n141, Translation="0,-5,5", Rotation="20,0,-30", Scale="1,1,1") {
    rotationMode=BMM_Additive
    scaleMode=BMM_Additive
    translationMode=BMM_Additive
    alphaBoolBlend {blendOption=Linear}
    alphaScaleBiasClamp {bInterpResult=true, interpSpeedDecreasing=3, interpSpeedIncreasing=3}
    boneToModify {boneName=dyn_coat_r_01}
    rotation {pitch=20, roll=-30}
    translation {y=-5, z=5}
   }
   n146 = "Transform (Modify) Bone - Bone: dyn_coat_bk_r_01"(ComponentPose=@n145, Alpha=@n141, Translation="0,0,0", Rotation="0,0,-30", Scale="1,1,1") {
    rotationMode=BMM_Additive
    scaleMode=BMM_Additive
    translationMode=BMM_Additive
    alphaBoolBlend {blendOption=Linear}
    alphaScaleBiasClamp {bInterpResult=true, interpSpeedDecreasing=3, interpSpeedIncreasing=3}
    boneToModify {boneName=dyn_coat_bk_r_01}
    rotation {roll=-30}
   }
   n147 = Get OnZiplineFloat() -> (OnZiplineFloat:double)
   n148 = "Transform (Modify) Bone - Bone: dyn_coat_bk_r_01"(ComponentPose=@n146, Alpha=@n147, Translation="0,0,0", Rotation="0,0,30", Scale="1,1,1") {
    rotationMode=BMM_Additive
    scaleMode=BMM_Additive
    translationMode=BMM_Additive
    alphaBoolBlend {blendOption=Linear}
    alphaScaleBiasClamp {bInterpResult=true, interpSpeedDecreasing=9, interpSpeedIncreasing=9}
    boneToModify {boneName=dyn_coat_bk_r_01}
    rotation {roll=30}
   }
   n149 = "Transform (Modify) Bone - Bone: dyn_coat_bk_l_01"(ComponentPose=@n148, Alpha=@n147, Translation="0,0,0", Rotation="15,0,0", Scale="1,1,1") {
    rotationMode=BMM_Additive
    scaleMode=BMM_Additive
    translationMode=BMM_Additive
    alphaBoolBlend {blendOption=Linear}
    alphaScaleBiasClamp {bInterpResult=true, interpSpeedDecreasing=9, interpSpeedIncreasing=9}
    boneToModify {boneName=dyn_coat_bk_l_01}
    rotation {pitch=15}
   }
   n150 = "Transform (Modify) Bone - Bone: dyn_coat_bk_01"(ComponentPose=@n149, Alpha=@n147, Translation="0,0,0", Rotation="0,0,0", Scale="1,1,1") {
    rotationMode=BMM_Additive
    scaleMode=BMM_Additive
    translationMode=BMM_Additive
    alphaBoolBlend {blendOption=Linear}
    alphaScaleBiasClamp {bInterpResult=true, interpSpeedDecreasing=9, interpSpeedIncreasing=9}
    boneToModify {boneName=dyn_coat_bk_01}
   }
   n151 = Get AlphaIs_Floating() -> (AlphaIs_Floating:double)
   n152 = "Transform (Modify) Bone - Bone: dyn_coat_bk_l_01"(ComponentPose=@n150, Alpha=@n151, Translation="1,0,0", Rotation="0,20,7", Scale="1,1,1") {
    rotationMode=BMM_Additive
    scaleMode=BMM_Additive
    translationMode=BMM_Additive
    alphaBoolBlend {blendOption=Linear}
    alphaScaleBiasClamp {bInterpResult=true, interpSpeedDecreasing=9, interpSpeedIncreasing=9}
    boneToModify {boneName=dyn_coat_bk_l_01}
    rotation {roll=7, yaw=20}
    translation {x=1}
   }
   n153 = "Transform (Modify) Bone - Bone: dyn_coat_l_01"(ComponentPose=@n152, Alpha=@n151, Translation="5,0,0", Rotation="0,0,0", Scale="1,1,1") {
    rotationMode=BMM_Additive
    scaleMode=BMM_Additive
    translationMode=BMM_Additive
    alphaBoolBlend {blendOption=Linear}
    alphaScaleBiasClamp {bInterpResult=true, interpSpeedDecreasing=9, interpSpeedIncreasing=9}
    boneToModify {boneName=dyn_coat_l_01}
    translation {x=5}
   }
   n154 = "Transform (Modify) Bone - Bone: dyn_coat_r_01"(ComponentPose=@n153, Alpha=@n151, Translation="0,0,0", Rotation="0,-7,0", Scale="1,1,1") {
    rotationMode=BMM_Additive
    scaleMode=BMM_Additive
    translationMode=BMM_Additive
    alphaBoolBlend {blendOption=Linear}
    alphaScaleBiasClamp {bInterpResult=true, interpSpeedDecreasing=9, interpSpeedIncreasing=9}
    boneToModify {boneName=dyn_coat_r_01}
    rotation {yaw=-7}
   }
   n155 = "Transform (Modify) Bone - Bone: dyn_coat_bk_r_01"(ComponentPose=@n154, Alpha=@n151, Translation="0,0,2", Rotation="0,0,3", Scale="1,1,1") {
    rotationMode=BMM_Additive
    scaleMode=BMM_Additive
    translationMode=BMM_Additive
    alphaBoolBlend {blendOption=Linear}
    alphaScaleBiasClamp {bInterpResult=true, interpSpeedDecreasing=9, interpSpeedIncreasing=9}
    boneToModify {boneName=dyn_coat_bk_r_01}
    rotation {roll=3}
    translation {z=2}
   }
   Save cached pose 'DetectedCrouchingCache'(Pose=Component To Local(ComponentPose=@n155)) {cacheName=DetectedCrouchingCache}
  ]}
 EventGraph {nodes [
   n1 = Event BlueprintUpdateAnimation()
   > n2 = Sequence!(execute) -> (then_0, then_1, then_2:Exec, then_3:Exec, then_4:Exec, then_5:Exec, then_6:Exec, then_7:Exec, then_8:Exec, then_9:Exec)
   n3 = GetOwnerSkeletalMeshComponent()
   n4 = BreakRotator(InRot=BreakTransform(InTransform=GetDeltaTransformFromRefPose(self=@n3, BoneName=thigh_r, BaseName=pelvis)).Rotation)
   > Set thigh_rot_up_r(thigh_rot_up_r:double=MapRangeClamped(Value=@n4.Yaw, InRangeA=-20, InRangeB=-100)) -> (Output_Get:double)
   > Set thigh_rot_otr_r(thigh_rot_otr_r:double=MapRangeClamped(Value=@n4.Pitch, InRangeA=0, InRangeB=30)) -> (Output_Get:double)
   n5 = GetOwnerSkeletalMeshComponent()
   n6 = BreakRotator(InRot=BreakTransform(InTransform=GetDeltaTransformFromRefPose(self=@n5, BoneName=thigh_l, BaseName=pelvis)).Rotation)
   > Set thigh_rot_up_l(thigh_rot_up_l:double=MapRangeClamped(Value=@n6.Yaw, InRangeA=-10, InRangeB=-100)) -> (Output_Get:double)
   > Set thigh_rot_inr_l(thigh_rot_inr_l:double=MapRangeClamped(Value=@n6.Pitch, InRangeA=0, InRangeB=-15)) -> (Output_Get:double)
   > Set thigh_rot_otr_l(thigh_rot_otr_l:double=MapRangeClamped(Value=@n6.Pitch, InRangeA=20, InRangeB=35)) -> (Output_Get:double)
   > Set IsLegRotationRight(IsLegRotationRight:double=MapRangeClamped(Value=@n4.Yaw, InRangeA=-20, InRangeB=-100, OutRangeA=1, OutRangeB=0)) -> (Output_Get:double)
   > Set IsLegRotationLeft(IsLegRotationLeft:double=MapRangeClamped(Value=@n6.Yaw, InRangeA=-10, InRangeB=-100, OutRangeA=1, OutRangeB=0)) -> (Output_Get:double)
   n7 = Get bIsMoving2D()
   Set inCrouchingNoMoving(execute=@n2.then_1, inCrouchingNoMoving:bool=AND Boolean(A=Get bIsCrouching(), B=NOT Boolean(A=@n7))) -> (Output_Get:bool)
   Set DeltaTimeLocal(execute=@n2.then_2, DeltaTimeLocal:double=@n1.DeltaTimeX) -> (Output_Get:double)
   > n8 = Branch(Condition=Get bClothEnabled())
   > n9 = Sequence!(execute) -> (then_0, then_1)
   n10 = Gravity Override State Function DEPRECATED(
    execute=[@n9.then_0, @n8.else]
    CharacterPartAnimInstance=Self-Reference()
    GameTimeInSeconds=GetGameTimeInSeconds()
    GravityOverrideParams=Get Gravity Override Params()
   )
   > Set GravityOverride(GravityOverride=@n10)
   n11 = Cloth Controller Function DEPRECATED!(
    execute=@n9.then_1
    CharacterPartAnimInstance:Custom Character Part Anim Instance Object Reference=Self-Reference()
    ClothingAssetName:String=Ultralight_Cloth
    Cloth Paramaters:Cloth Params Struct Structure=Get Cloth Paramaters_0() -> (Cloth Paramaters_0:Cloth Params Struct Structure)
    Cloth Core Settings:Cloth Core Setting Struct Structure=Get Cloth Core Settings_0() -> (Cloth Core Settings_0:Cloth Core Setting Struct Structure)
    ClothUpAxis:Vector="1,0,0"
    OnGround_Moving_SpeedLimit:double=2
    OnGround_Moving_AngularVel_Max:double=0.05
    OnGround_Moving_FictitiousVel_Max:double=0.6
    LinVelScale_FwdBwd_Clamp:double=0
    LinVelScale_UpDwn_Clamp:double=0.7
    DeltaTime:double=Get DeltaTimeLocal() -> (DeltaTimeLocal:double)
    GameTimeInSeconds:double=GetGameTimeInSeconds()
    WindGustsParams:Wind Gust Params Struct Structure=Get Wind Gust Params() -> (Wind Gust Params:Wind Gust Params Struct Structure)
    WindGustLocalSettings:Wind Gust Local Setting Struct Structure=Get Gust Local Settings() -> (Gust Local Settings:Wind Gust Local Setting Struct Structure)
   ) -> (
    then
    CurrentStateOut:String
    Cloth Settings Out:Cloth Setting Struct Structure
    Cloth Core Settings Out:Cloth Core Setting Struct Structure
    WindGustSettings Out:Wind Gust Setting Struct Structure
    WindGustLocalSettings Out:Wind Gust Local Setting Struct Structure
    GustStrength - WindVelocityOut:Vector
    LinearVelVector-DELETE LATER:Vector
   )
   > Set Gust Local Settings(Gust Local Settings:Wind Gust Local Setting Struct Structure=@n11.WindGustLocalSettings Out) -> (Output_Get:Wind Gust Local Setting Struct Structure)
   n12 = Is Valid(exec=[@n2.then_3, @n2.then_4], InputObject=Self-Reference())
   n13 = CheckForBackbling(LongBackblingThreshold=40)
   n14 = Branch(execute=[@n12.Is Valid, @n61.Completed], Condition=@n13)
   > n15 = Set IsBackblingOn(IsBackblingOn:bool=true) -> (Output_Get:bool)
   > n16 = Branch(Condition=@n13.bIsBackblingACape)
   > Set Is Backbling A Cape(Is Backbling A Cape:bool=true) -> (Output_Get:bool)
   > Set Cape Alpha(Cape Alpha:double=0) -> (Output_Get:double)
   n17 = GetPartSkeletalMeshComponent()
   > Set ClothBlendWeight!(execute, ClothBlendWeight:float=0, self:Skeletal Mesh Component Object Reference=@n17) -> (then, Output_Get:float)
   > Do Once(Reset=@n19.then)
   > n18 = SuspendClothingSimulation(self:Skeletal Mesh Component Object Reference=@n17)
   Do Once(execute=@n20.then, Reset=@n18.then)
   > n19 = ResumeClothingSimulation(self:Skeletal Mesh Component Object Reference=@n17)
   Set Is Backbling A Cape(execute=@n16.else, Is Backbling A Cape=false)
   > Set Cape Alpha(Cape Alpha=1)
   > n20 = Set ClothBlendWeight(ClothBlendWeight=1, self=@n17)
   n21 = Set IsBackblingOn(execute=@n14.else, IsBackblingOn=false)
   n22 = Get bIsMoving2D()
   n23 = Branch(execute=@n2.then_5, Condition=AND Boolean(A=Get bIsCrouching(), B=NOT Boolean(A=@n22)))
   > Set IsCrouchingIdle(IsCrouchingIdle:double=1) -> (Output_Get:double)
   Set IsCrouchingIdle(execute=@n23.else, IsCrouchingIdle=0)
   n24 = Branch(execute=@n2.then_6, Condition=AND Boolean(A=Get bIsCrouching(), B=Get bIsMoving2D()))
   > Set IsCrouchingMoving(IsCrouchingMoving:double=1) -> (Output_Get:double)
   Set IsCrouchingMoving(execute=@n24.else, IsCrouchingMoving=0)
   n25 = Get bIsSkyDiving()
   n26 = Get bIsSurfaceSwimming()
   n27 = Get bIsCrouching()
   n28 = Branch(
    execute=@n2.then_7
    Condition=AND Boolean!(A=Get bIsMoving2D(), B=NOT Boolean(A=@n27), C:bool=NOT Boolean(A=@n25), D:bool=NOT Boolean(A=@n26)) -> (ReturnValue)
   )
   > Set IsStandMoving(IsStandMoving:double=1) -> (Output_Get:double)
   Set IsStandMoving(execute=@n28.else, IsStandMoving=0)
   n29 = NOT Boolean(A=Get bIsSliding())
   n30 = Get bIsCrouching()
   n31 = Branch(execute=@n2.then_8, Condition=OR Boolean(A=AND Boolean!(A=@n30, B=@n29) -> (ReturnValue), B=Get bIsDBNO()))
   > Set IsCrouchingDetected(IsCrouchingDetected:bool=true) -> (Output_Get:bool)
   Set IsCrouchingDetected(execute=@n31.else, IsCrouchingDetected=false)
   n32 = Sequence!(execute=@n2.then_9) -> (then_0, then_1, then_2:Exec, then_3:Exec, then_4:Exec, then_5:Exec, then_6:Exec)
   n33 = Branch(execute=@n32.then_1, Condition=Get bIsDBNO())
   > Set Is_DBNO(Is_DBNO:float=1) -> (Output_Get:float)
   Set Is_DBNO(execute=@n33.else, Is_DBNO=0)
   n34 = PlayerMovement(execute=@n32.then_5, AnimInstance=Self-Reference(), XDotProductCoefficient=100, YDotProductCoefficient=100, ZDotProductCoefficient=100)
   n35 = BreakVector(InVec=MakeVector(X=@n34, Y=@n34.YMovement, Z=@n34.ZMovement))
   > n36 = Branch(Condition=AND Boolean(A="float > float"(A=@n35.Z), B=Get bIsSkyDiving()))
   > Set Alpha_isSkidiveUp(Alpha_isSkidiveUp:float=1) -> (Output_Get:float)
   Set Alpha_isSkidiveUp(execute=@n36.else, Alpha_isSkidiveUp=0)
   n37 = Branch(execute=@n32.then_3, Condition=Get bIsSliding())
   > Set IsSliding(IsSliding:double=1) -> (Output_Get:double)
   Set IsSliding(execute=@n37.else, IsSliding=0)
   n38 = Branch(execute=@n32.then_2, Condition=AND Boolean(A=Get AlphaIsWeapons_bool() -> (AlphaIsWeapons_bool:bool), B=Get bIsCrouching()))
   > Set IsCrouchingWithWeapons(IsCrouchingWithWeapons:double=1) -> (Output_Get:double)
   Set IsCrouchingWithWeapons(execute=@n38.else, IsCrouchingWithWeapons=0)
   n39 = Branch(execute=@n32.then_4, Condition=AND Boolean(A=Get bIsSkyDiving(), B=Get bIsParachuteOpened()))
   > Set IsSkyDiving(IsSkyDiving:double=1) -> (Output_Get:double)
   Set IsSkyDiving(execute=@n39.else, IsSkyDiving=0)
   n40 = Sequence(execute=@n32.then_6)
   n41 = Branch(execute=@n40.then_6, Condition=Get bIsSkydiveFloating())
   > Set AlphaIs_Floating(AlphaIs_Floating:double=1) -> (Output_Get:double)
   Set AlphaIs_Floating(execute=@n41.else, AlphaIs_Floating=0)
   n42 = Get bIsMoving2D()
   n43 = Branch(execute=@n40.then_1, Condition=AND Boolean(A=Get bIsOnGround() -> (bIsOnGround:bool), B=NOT Boolean(A=@n42)))
   > Set Alpha_isStandingIddle(Alpha_isStandingIddle:float=1) -> (Output_Get:float)
   Set Alpha_isStandingIddle(execute=@n43.else, Alpha_isStandingIddle=0)
   n44 = Branch(execute=@n40.then_2, Condition=Get bIsParachuteOpened())
   > Set Alpha_isParachute(Alpha_isParachute:float=1) -> (Output_Get:float)
   Set Alpha_isParachute(execute=@n44.else, Alpha_isParachute=0)
   n45 = GetOwnerSkeletalMeshComponent()
   n46 = BreakRotator(InRot=BreakTransform(InTransform=GetDeltaTransformFromRefPose(self=@n45, BoneName=calf_r, BaseName=pelvis)).Rotation)
   n47 = Branch(execute=@n40.then_3, Condition=AND Boolean(A="float > float"(A=@n46, B=22), B=Get OnRailMode() -> (OnRailMode:bool)))
   > Set Alpha_isRotateLegR_OnRail(Alpha_isRotateLegR_OnRail:float=1) -> (Output_Get:float)
   Set Alpha_isRotateLegR_OnRail(execute=@n47.else, Alpha_isRotateLegR_OnRail=0)
   n48 = Get bIsSurfaceSwimming()
   n49 = Get bIsOnGround()
   n50 = Get bIsJumping()
   n51 = Get bIsSkyDiving()
   n52 = Get bIsFalling()
   n53 = Get OnZipline() -> (OnZipline:bool)
   n54 = Branch(
    execute=@n40.then_0
    Condition=AND Boolean!(
     A=NOT Boolean(A=@n49)
     B=NOT Boolean(A=@n50)
     C:bool=NOT Boolean(A=@n51)
     D:bool=NOT Boolean(A=@n52)
     E:bool=NOT Boolean(A=@n48)
     F:bool=NOT Boolean(A=@n53)
    ) -> (ReturnValue)
   )
   > Set OnRailMode(OnRailMode:bool=true) -> (Output_Get:bool)
   > Set Is_GrindRail(Is_GrindRail:float=1) -> (Output_Get:float)
   Set OnRailMode(execute=@n54.else, OnRailMode=false)
   > Set Is_GrindRail(Is_GrindRail=0)
   n55 = Branch(execute=@n40.then_4, Condition=Get bIsSkydiveDiveMode())
   > Set AlphaIs_SkydiveMode(AlphaIs_SkydiveMode:double=1) -> (Output_Get:double)
   Set AlphaIs_SkydiveMode(execute=@n55.else, AlphaIs_SkydiveMode=0)
   n56 = Branch(execute=@n40.then_5, Condition=AND Boolean!(A=Get bIsPlayingMeleeAnim(), B=Get bIsCrouching(), C=Get bIsMoving2D()) -> (ReturnValue))
   > Set AlphaIs_CrouchMelee(AlphaIs_CrouchMelee:double=1) -> (Output_Get:double)
   Set AlphaIs_CrouchMelee(execute=@n56.else, AlphaIs_CrouchMelee=0)
   n57 = Cast To FortPlayerAnimInstance(execute=@n32.then_0, Object=GetAnimInstance(self=GetOwnerSkeletalMeshComponent()))
   > Set ar_firingPosition(ar_firingPosition:int=1) -> (Output_Get:int=0)
   > n58 = Switch on EFortWeaponCoreAnimation!(execute, Selection:EFortWeaponCoreAnimation Enum=Get WeaponCoreAnim(self=@n57)) -> (
    Melee:Exec
    Pistol:Exec
    Shotgun:Exec
    PaperBlueprint:Exec
    Rifle:Exec
    MeleeOneHand:Exec
    MachinePistol:Exec
    RocketLauncher:Exec
    GrenadeLauncher:Exec
    GoingCommando:Exec
    AssaultRifle:Exec
    TacticalShotgun:Exec
    SniperRifle:Exec
    TrapPlacement:Exec
    ShoulderLauncher:Exec
    AbilityDecoTool:Exec
    Crossbow:Exec
    C4:Exec
    RemoteControl:Exec
    DualWield:Exec
    AR_BullPup:Exec
    AR_ForwardGrip:Exec
    MedPackPaddles:Exec
    SMG_P90:Exec
    AR_DrumGun:Exec
    Consumable_Small:Exec
    Consumable_Large:Exec
    Balloon:Exec
    MountedTurret:Exec
    CreativeTool:Exec
    ExplosiveBow:Exec
    AshtonIndigo:Exec
    AshtonChicago:Exec
    MeleeDualWield:Exec
    Unarmed:Exec
   )
   Set AlphaIsWeapons_bool(
    execute=[
     @n58.AR_BullPup
     @n58.AR_DrumGun
     @n58.AR_ForwardGrip
     @n58.AbilityDecoTool
     @n58.AshtonChicago
     @n58.AshtonIndigo
     @n58.Balloon
     @n58.C4
     @n58.Consumable_Large
     @n58.Consumable_Small
     @n58.CreativeTool
     @n58.Crossbow
     @n58.DualWield
     @n58.ExplosiveBow
     @n58.GrenadeLauncher
     @n58.MachinePistol
     @n58.MedPackPaddles
     @n58.Melee
     @n58.MeleeDualWield
     @n58.MeleeOneHand
     @n58.MountedTurret
     @n58.PaperBlueprint
     @n58.Pistol
     @n58.RemoteControl
     @n58.Rifle
     @n58.SMG_P90
     @n58.Shotgun
     @n58.ShoulderLauncher
     @n58.SniperRifle
     @n58.TacticalShotgun
     @n58.TrapPlacement
     @n58.Unarmed
    ]
    AlphaIsWeapons_bool:bool=false
   ) -> (Output_Get:bool)
   Set AlphaIsWeapons_bool(execute=[@n58.AssaultRifle, @n58.GoingCommando, @n58.RocketLauncher], AlphaIsWeapons_bool=true)
   n59 = Event BlueprintProcessAnimInputEvent()
   > n60 = Branch(Condition=IsNamed(self=@n59.AnimInputEvent, InEventName=aie.Player.Zipline.Entered))
   > Set OnZipline(OnZipline:bool=true) -> (Output_Get:bool)
   > Set OnZiplineFloat(OnZiplineFloat:double=1) -> (Output_Get:double)
   Branch(execute=@n60.else, Condition=IsNamed(self=@n59.AnimInputEvent, InEventName=aie.Player.Zipline.Exited))
   > Set OnZipline(OnZipline=false)
   > Set OnZiplineFloat(OnZiplineFloat=0)
   n61 = Do Once()
   DrawVectorFromJointFunction!(
    execute
    SkelMeshComponent:Scene Component Object Reference
    JointName:Name=pelvis
    LineEnd:Vector=Break Cloth Setting Struct!(ClothSettingStruct:"Cloth Setting Struct Structure (by ref)"=@n11.Cloth Settings Out) -> (
     LinearVelocityScale_7_#1:Vector=~
     AngularVelocityScale_6_#2:float=~
     FictitiousAngularScale_10_#3:float=~
     GravityOverride_15_#4:Vector=~
     DragCoefficient_23_#5:Vector 2D Structure
     LiftCoefficient_24_#6:Vector 2D Structure
     JointName_27_#7:Name
    ) {is=title K2Node_BreakStruct}.GravityOverride_15_#4
    JointOffset:Vector="0, 0, 0"
    LineColor:Linear Color Structure=(R=1,G=0,B=0,A=1)
    Thickness:double=2
    Foreground:EDrawDebugSceneDepthPriorityGroup Enum=Foreground
   ) -> (then)
   Set IsBackblingOnAlpha(IsBackblingOnAlpha:double="To Float (Boolean)"(InBool=@n15)) -> (Output_Get:double)
   Set IsBackblingOnAlpha(IsBackblingOnAlpha="To Float (Boolean)"(InBool=@n21))
  ]}
]
variables [
 Alpha_isParachute {type=number}
 Alpha_isRotateLegR_OnRail {type=number}
 Alpha_isSkidiveUp {type=number}
 Alpha_isStandingIddle {type=number}
 AlphaIs_CrouchMelee {type=number}
 AlphaIs_Floating {type=number}
 AlphaIs_SkydiveMode {type=number}
 AlphaIsWeapons_bool {type=boolean}
 ar_firingPosition {type=integer}
 backbling_equipped {type=boolean}
 Cape Alpha {type=number}
 cape_disabled {type=boolean}
 Cloth Core Settings_0 {
  type=ClothCoreSettingStruct
  default {
   collisionThickness=0
   dampingCoefficient=0.1
   frictionCoefficient=0.3
   localDampingCoefficient=0.1
   numberIterations=2
   numberSubsteps=1
   selfCollisionThickness=0
   animDriveDamping=<0 1>
   animDriveStiffness=<0 1>
   areaStiffness=<1 1>
   bendingStiffness=<1 1>
   edgeStiffness=<1 1>
   tetherStiffness=<1 1>
  }
 }
 Cloth Paramaters_0 {
  type=ClothParamsStruct
  default {
   dBNO {
    angularVelocityScale=0.6
    fictitiousAngularScale=0.1
    jointName=pelvis
    dragCoefficient=<0.2 1>
    gravityOverride=<-980 400 0>
    liftCoefficient=<0 0.5>
    linearVelocityScale=<0.3 0.3 0.1>
   }
   dBNO_Carried {
    angularVelocityScale=0.6
    fictitiousAngularScale=0.1
    jointName=pelvis
    dragCoefficient=<0.2 1>
    gravityOverride=<-980 -200 0>
    liftCoefficient=<0 0.5>
    linearVelocityScale=<0.3 0.3 0.1>
   }
   dBNO_Impulse {
    angularVelocityScale=0
    fictitiousAngularScale=0
    jointName=pelvis
    dragCoefficient=<0 0.2>
    gravityOverride=<-3000 -2000 0>
    liftCoefficient=<0 0.1>
    linearVelocityScale=<0.3 0.1 0.3>
   }
   dBNO_RollBwd {
    angularVelocityScale=0
    fictitiousAngularScale=0
    jointName=pelvis
    dragCoefficient=<0 0.2>
    gravityOverride=<-2000 0 0>
    liftCoefficient=<0 0.1>
    linearVelocityScale=<0.3 0.1 0.3>
   }
   dBNO_RollFwd {
    angularVelocityScale=0
    fictitiousAngularScale=0
    jointName=pelvis
    dragCoefficient=<0 0.2>
    gravityOverride=<-980 0 0>
    liftCoefficient=<0 0.1>
    linearVelocityScale=<0.3 0.1 0.3>
   }
   dBNO_Tumbling {
    angularVelocityScale=0
    fictitiousAngularScale=0
    jointName=pelvis
    dragCoefficient=<0 0.2>
    gravityOverride=<-2500 -300 0>
    liftCoefficient=<0 0.1>
    linearVelocityScale=<0.3 0.1 0.3>
   }
   emoteOrMelee {
    angularVelocityScale=0.8
    fictitiousAngularScale=0.2
    jointName=pelvis
    dragCoefficient=<0 0.3>
    gravityOverride=<-980 -300 0>
    liftCoefficient=<0 0.3>
    linearVelocityScale=<1 1 1>
   }
   front_End {
    angularVelocityScale=0.1
    fictitiousAngularScale=0.2
    jointName=root
    dragCoefficient=<0.2 0.2>
    gravityOverride=<0 -200 -980>
    liftCoefficient=<0.1 0.1>
    linearVelocityScale=<0.1 0.1 0.1>
   }
   onGround_Crouch_Moving {
    angularVelocityScale=0.1
    fictitiousAngularScale=0.3
    jointName=root
    dragCoefficient=<2 2>
    gravityOverride=<0 0 -980>
    liftCoefficient=<0.6 0.6>
    linearVelocityScale=<0.6 0.6 0.5>
   }
   onGround_Falling {
    angularVelocityScale=0.6
    fictitiousAngularScale=0
    jointName=root
    dragCoefficient=<0 0.2>
    gravityOverride=<0 -500 -980>
    liftCoefficient=<0 0.1>
    linearVelocityScale=<2 1 0.3>
   }
   onGround_Moving {
    angularVelocityScale=0.6
    fictitiousAngularScale=0.1
    jointName=root
    dragCoefficient=<0 0.3>
    gravityOverride=<0 -100 -980>
    liftCoefficient=<0 0.2>
    linearVelocityScale=<0.6 1 1>
   }
   onGround_Standing {
    angularVelocityScale=0.05
    fictitiousAngularScale=0.6
    jointName=root
    dragCoefficient=<0 2>
    gravityOverride=<0 -100 -980>
    liftCoefficient=<0 0.5>
    linearVelocityScale=<0.6 1 1>
   }
   rail_Grinding {
    angularVelocityScale=0.2
    fictitiousAngularScale=0
    jointName=root
    dragCoefficient=<0.1 0.2>
    gravityOverride=<50 -500 -400>
    liftCoefficient=<0 0.5>
    linearVelocityScale=<0.1 0.1 0.1>
   }
   riding_Moving {
    angularVelocityScale=0.05
    fictitiousAngularScale=1
    jointName=root
    dragCoefficient=<0.5 0.5>
    gravityOverride=<0 0 -980>
    liftCoefficient=<0.5 0.5>
    linearVelocityScale=<0.4 0.6 0.3>
   }
   riding_Moving_Targeting {
    angularVelocityScale=0.05
    fictitiousAngularScale=0.7
    jointName=root
    dragCoefficient=<0.5 0.5>
    gravityOverride=<0 0 -980>
    liftCoefficient=<0.5 0.5>
    linearVelocityScale=<0.2 0.6 0.2>
   }
   riding_Standing {
    angularVelocityScale=0.05
    fictitiousAngularScale=1
    jointName=root
    dragCoefficient=<0.5 0.5>
    gravityOverride=<0 -450 -980>
    liftCoefficient=<0.5 0.5>
    linearVelocityScale=<0.6 0.6 0.3>
   }
   skydive_DivingDown {
    angularVelocityScale=0.1
    fictitiousAngularScale=0
    jointName=pelvis
    dragCoefficient=<0 1.5>
    gravityOverride=<-450 -100 0>
    liftCoefficient=<0 0.3>
    linearVelocityScale=<0.1 0.1 0.1>
   }
   skydive_FloatingDown {
    angularVelocityScale=0.1
    fictitiousAngularScale=0
    jointName=root
    dragCoefficient=<0 2.5>
    gravityOverride=<0 0 980>
    liftCoefficient=<0 0.4>
    linearVelocityScale=<0.2 0.2 0.1>
   }
   skydive_GoingUp {
    angularVelocityScale=0.1
    fictitiousAngularScale=0
    jointName=root
    dragCoefficient=<0 0>
    gravityOverride=<0 -3000 3000>
    liftCoefficient=<0 0>
    linearVelocityScale=<0.1 0.1 0.05>
   }
   skydive_Parachuting {
    angularVelocityScale=0.1
    fictitiousAngularScale=0
    jointName=root
    dragCoefficient=<0 2.5>
    gravityOverride=<0 0 -980>
    liftCoefficient=<0 0.5>
    linearVelocityScale=<0.2 0.2 0.2>
   }
   skydive_Surfing {
    angularVelocityScale=0.6
    fictitiousAngularScale=0.1
    jointName=root
    dragCoefficient=<1.5 1.5>
    gravityOverride=<0 -600 -500>
    liftCoefficient=<1 1>
    linearVelocityScale=<0.1 0.1 0.3>
   }
   swimming {
    angularVelocityScale=0.6
    fictitiousAngularScale=0
    jointName=root
    dragCoefficient=<0 2>
    gravityOverride=<0 -600 400>
    liftCoefficient=<0 0.4>
    linearVelocityScale=<0.6 0.2 0.2>
   }
   tactical_Sprint {
    angularVelocityScale=0.6
    fictitiousAngularScale=0.1
    jointName=root
    dragCoefficient=<2 2>
    gravityOverride=<0 0 -980>
    liftCoefficient=<0.5 0.5>
    linearVelocityScale=<0.6 0.6 0.5>
   }
   ziplining {
    angularVelocityScale=0.6
    fictitiousAngularScale=0
    jointName=root
    dragCoefficient=<0 1.5>
    gravityOverride=<0 0 -980>
    liftCoefficient=<0 0.3>
    linearVelocityScale=<0.4 0.4 0.5>
   }
  }
 }
 DeltaTimeLocal {type=number}
 Gravity Override Params {
  type=GravityOverrideParamsStruct
  default {skydive_FloatingDown {gravityOverride=<0 -700 400>, windAmplitude=<200 -500 200>, windOffset=<0 -600 200>}}
 }
 GravityOverride {type=Vector}
 Gust Local Settings {type=WindGustLocalSettingStruct, default {elapsedTime=0, finalGustStrengthLocal=0, gustDuration=0, gustStrength=0, timeBetweenGusts=0}}
 inCrouchingNoMoving {type=boolean}
 Is Backbling A Cape {type=boolean}
 Is_DBNO {type=number}
 Is_GrindRail {type=number}
 IsBackblingOn {type=boolean}
 IsBackblingOnAlpha {type=number}
 IsCrouchingDetected {type=boolean}
 IsCrouchingIdle {type=number}
 IsCrouchingMoving {type=number}
 IsCrouchingWithWeapons {type=number}
 IsLegRotationLeft {type=number}
 IsLegRotationRight {type=number}
 IsSkyDiving {type=number}
 IsSliding {type=number}
 IsStandMoving {type=number}
 Leg L Rotate Bool {type=boolean}
 OnRailMode {type=boolean}
 OnZipline {type=boolean}
 OnZiplineFloat {type=number}
 Skeletal Mesh Component {type=SkinnedMeshComponent, default=null}
 thigh_rot_inr_l {type=number}
 thigh_rot_otr_l {type=number}
 thigh_rot_otr_r {type=number}
 thigh_rot_up_l {type=number}
 thigh_rot_up_r {type=number}
 Wind Gust Params {
  type=WindGustParamsStruct
  default {
   dBNO {
    maxDownInterpSpeed=5
    maxGustDuration=2
    maxGustStrength=7
    maxTimeBetweenGusts=7
    maxUpInterpSpeed=5
    minDownInterpSpeed=2
    minGustDuration=0.5
    minGustStrength=3
    minTimeBetweenGusts=0
    minUpInterpSpeed=2
    windTurbulenceScalarA=0.1
    windTurbulenceScalarB=1
    windAmplitude=<0 0 0>
    windFrequency=<0 0 0>
    windOffset=<0 0 0>
   }
   dBNO_Carried {
    maxDownInterpSpeed=5
    maxGustDuration=2
    maxGustStrength=7
    maxTimeBetweenGusts=7
    maxUpInterpSpeed=5
    minDownInterpSpeed=2
    minGustDuration=0.5
    minGustStrength=3
    minTimeBetweenGusts=0
    minUpInterpSpeed=2
    windTurbulenceScalarA=0.1
    windTurbulenceScalarB=1
    windAmplitude=<0 0 0>
    windFrequency=<0 0 0>
    windOffset=<0 0 0>
   }
   dBNO_Impulse {
    maxDownInterpSpeed=5
    maxGustDuration=2
    maxGustStrength=7
    maxTimeBetweenGusts=7
    maxUpInterpSpeed=5
    minDownInterpSpeed=2
    minGustDuration=0.5
    minGustStrength=3
    minTimeBetweenGusts=0
    minUpInterpSpeed=2
    windTurbulenceScalarA=0.1
    windTurbulenceScalarB=1
    windAmplitude=<0 0 0>
    windFrequency=<0 0 0>
    windOffset=<0 0 0>
   }
   dBNO_RollBwd {
    maxDownInterpSpeed=5
    maxGustDuration=2
    maxGustStrength=7
    maxTimeBetweenGusts=7
    maxUpInterpSpeed=5
    minDownInterpSpeed=2
    minGustDuration=0.5
    minGustStrength=3
    minTimeBetweenGusts=0
    minUpInterpSpeed=2
    windTurbulenceScalarA=0.1
    windTurbulenceScalarB=1
    windAmplitude=<0 0 0>
    windFrequency=<0 0 0>
    windOffset=<0 0 0>
   }
   dBNO_RollFwd {
    maxDownInterpSpeed=5
    maxGustDuration=2
    maxGustStrength=7
    maxTimeBetweenGusts=7
    maxUpInterpSpeed=5
    minDownInterpSpeed=2
    minGustDuration=0.5
    minGustStrength=3
    minTimeBetweenGusts=0
    minUpInterpSpeed=2
    windTurbulenceScalarA=0.1
    windTurbulenceScalarB=1
    windAmplitude=<0 0 0>
    windFrequency=<0 0 0>
    windOffset=<0 0 0>
   }
   dBNO_Tumbling {
    maxDownInterpSpeed=5
    maxGustDuration=2
    maxGustStrength=7
    maxTimeBetweenGusts=7
    maxUpInterpSpeed=5
    minDownInterpSpeed=2
    minGustDuration=0.5
    minGustStrength=3
    minTimeBetweenGusts=0
    minUpInterpSpeed=2
    windTurbulenceScalarA=0.1
    windTurbulenceScalarB=1
    windAmplitude=<0 0 0>
    windFrequency=<0 0 0>
    windOffset=<0 0 0>
   }
   emoteOrMelee {
    maxDownInterpSpeed=5
    maxGustDuration=2
    maxGustStrength=7
    maxTimeBetweenGusts=7
    maxUpInterpSpeed=5
    minDownInterpSpeed=2
    minGustDuration=0.5
    minGustStrength=3
    minTimeBetweenGusts=0
    minUpInterpSpeed=2
    windTurbulenceScalarA=0.1
    windTurbulenceScalarB=1
    windAmplitude=<30 25 50>
    windFrequency=<2 3 5>
    windOffset=<-300 -150 0>
   }
   front_End {
    maxDownInterpSpeed=5
    maxGustDuration=2
    maxGustStrength=4
    maxTimeBetweenGusts=7
    maxUpInterpSpeed=5
    minDownInterpSpeed=2
    minGustDuration=0.5
    minGustStrength=2
    minTimeBetweenGusts=0
    minUpInterpSpeed=2
    windTurbulenceScalarA=0.1
    windTurbulenceScalarB=1
    windAmplitude=<5 15 25>
    windFrequency=<5 1.5 2>
    windOffset=<0 -100 -125>
   }
   onGround_Crouch_Moving {
    maxDownInterpSpeed=5
    maxGustDuration=2
    maxGustStrength=7
    maxTimeBetweenGusts=7
    maxUpInterpSpeed=5
    minDownInterpSpeed=2
    minGustDuration=0.5
    minGustStrength=3
    minTimeBetweenGusts=0
    minUpInterpSpeed=2
    windTurbulenceScalarA=0.1
    windTurbulenceScalarB=1
    windAmplitude=<100 25 75>
    windFrequency=<1.5 0.5 7.5>
    windOffset=<0 -100 -175>
   }
   onGround_Falling {
    maxDownInterpSpeed=5
    maxGustDuration=2
    maxGustStrength=7
    maxTimeBetweenGusts=7
    maxUpInterpSpeed=5
    minDownInterpSpeed=2
    minGustDuration=0.5
    minGustStrength=3
    minTimeBetweenGusts=0
    minUpInterpSpeed=2
    windTurbulenceScalarA=0.1
    windTurbulenceScalarB=1
    windAmplitude=<0 0 0>
    windFrequency=<0 0 0>
    windOffset=<0 0 0>
   }
   onGround_Moving {
    maxDownInterpSpeed=5
    maxGustDuration=8
    maxGustStrength=7
    maxTimeBetweenGusts=8
    maxUpInterpSpeed=5
    minDownInterpSpeed=2
    minGustDuration=0.2
    minGustStrength=3
    minTimeBetweenGusts=1
    minUpInterpSpeed=2
    windTurbulenceScalarA=0.1
    windTurbulenceScalarB=0.8
    windAmplitude=<30 65 42>
    windFrequency=<5 6 8.6>
    windOffset=<30 -250 -300>
   }
   onGround_Standing {
    maxDownInterpSpeed=5
    maxGustDuration=8
    maxGustStrength=7
    maxTimeBetweenGusts=8
    maxUpInterpSpeed=5
    minDownInterpSpeed=2
    minGustDuration=0.2
    minGustStrength=2
    minTimeBetweenGusts=1
    minUpInterpSpeed=2
    windTurbulenceScalarA=0.1
    windTurbulenceScalarB=0.8
    windAmplitude=<75 24 28>
    windFrequency=<1.5 2.3 3.7>
    windOffset=<50 -100 -120>
   }
   rail_Grinding {
    maxDownInterpSpeed=5
    maxGustDuration=8
    maxGustStrength=7
    maxTimeBetweenGusts=8
    maxUpInterpSpeed=5
    minDownInterpSpeed=2
    minGustDuration=0.2
    minGustStrength=3
    minTimeBetweenGusts=1
    minUpInterpSpeed=2
    windTurbulenceScalarA=0.1
    windTurbulenceScalarB=0.8
    windAmplitude=<20 65 42>
    windFrequency=<5 6 8.6>
    windOffset=<50 -500 -400>
   }
   riding_Moving {
    maxDownInterpSpeed=5
    maxGustDuration=2
    maxGustStrength=7
    maxTimeBetweenGusts=7
    maxUpInterpSpeed=5
    minDownInterpSpeed=2
    minGustDuration=0.5
    minGustStrength=3
    minTimeBetweenGusts=0
    minUpInterpSpeed=2
    windTurbulenceScalarA=0.1
    windTurbulenceScalarB=1
    windAmplitude=<0 0 0>
    windFrequency=<0 0 0>
    windOffset=<0 0 0>
   }
   riding_Moving_Targeting {
    maxDownInterpSpeed=5
    maxGustDuration=2
    maxGustStrength=7
    maxTimeBetweenGusts=7
    maxUpInterpSpeed=5
    minDownInterpSpeed=2
    minGustDuration=0.5
    minGustStrength=3
    minTimeBetweenGusts=0
    minUpInterpSpeed=2
    windTurbulenceScalarA=0.1
    windTurbulenceScalarB=1
    windAmplitude=<0 0 0>
    windFrequency=<0 0 0>
    windOffset=<0 0 0>
   }
   riding_Standing {
    maxDownInterpSpeed=5
    maxGustDuration=2
    maxGustStrength=7
    maxTimeBetweenGusts=7
    maxUpInterpSpeed=5
    minDownInterpSpeed=2
    minGustDuration=0.5
    minGustStrength=3
    minTimeBetweenGusts=0
    minUpInterpSpeed=2
    windTurbulenceScalarA=0.1
    windTurbulenceScalarB=1
    windAmplitude=<0 0 0>
    windFrequency=<0 0 0>
    windOffset=<0 0 0>
   }
   skydive_DivingDown {
    maxDownInterpSpeed=5
    maxGustDuration=8
    maxGustStrength=7
    maxTimeBetweenGusts=8
    maxUpInterpSpeed=5
    minDownInterpSpeed=2
    minGustDuration=0.2
    minGustStrength=3
    minTimeBetweenGusts=1
    minUpInterpSpeed=2
    windTurbulenceScalarA=0.1
    windTurbulenceScalarB=4
    windAmplitude=<0 25 35>
    windFrequency=<0 4 10>
    windOffset=<-450 -100 100>
   }
   skydive_FloatingDown {
    maxDownInterpSpeed=5
    maxGustDuration=8
    maxGustStrength=1
    maxTimeBetweenGusts=8
    maxUpInterpSpeed=5
    minDownInterpSpeed=2
    minGustDuration=0.2
    minGustStrength=0
    minTimeBetweenGusts=1
    minUpInterpSpeed=2
    windTurbulenceScalarA=0.5
    windTurbulenceScalarB=1
    windAmplitude=<50 20 35>
    windFrequency=<8 2 3>
    windOffset=<0 -300 200>
   }
   skydive_GoingUp {
    maxDownInterpSpeed=5
    maxGustDuration=8
    maxGustStrength=7
    maxTimeBetweenGusts=8
    maxUpInterpSpeed=5
    minDownInterpSpeed=2
    minGustDuration=0.2
    minGustStrength=3
    minTimeBetweenGusts=1
    minUpInterpSpeed=2
    windTurbulenceScalarA=0.1
    windTurbulenceScalarB=0.8
    windAmplitude=<0 0 0>
    windFrequency=<0 0 0>
    windOffset=<0 -500 300>
   }
   skydive_Parachuting {
    maxDownInterpSpeed=5
    maxGustDuration=8
    maxGustStrength=1
    maxTimeBetweenGusts=8
    maxUpInterpSpeed=5
    minDownInterpSpeed=2
    minGustDuration=5
    minGustStrength=0
    minTimeBetweenGusts=5
    minUpInterpSpeed=2
    windTurbulenceScalarA=0.2
    windTurbulenceScalarB=1
    windAmplitude=<25 50 75>
    windFrequency=<8 6 2>
    windOffset=<-150 -400 -150>
   }
   skydive_Surfing {
    maxDownInterpSpeed=5
    maxGustDuration=2
    maxGustStrength=7
    maxTimeBetweenGusts=7
    maxUpInterpSpeed=5
    minDownInterpSpeed=2
    minGustDuration=0.5
    minGustStrength=3
    minTimeBetweenGusts=0
    minUpInterpSpeed=2
    windTurbulenceScalarA=0.1
    windTurbulenceScalarB=1
    windAmplitude=<0 0 0>
    windFrequency=<0 0 0>
    windOffset=<0 0 0>
   }
   swimming {
    maxDownInterpSpeed=5
    maxGustDuration=2
    maxGustStrength=7
    maxTimeBetweenGusts=7
    maxUpInterpSpeed=5
    minDownInterpSpeed=2
    minGustDuration=0.5
    minGustStrength=3
    minTimeBetweenGusts=0
    minUpInterpSpeed=2
    windTurbulenceScalarA=0.1
    windTurbulenceScalarB=1
    windAmplitude=<10 25 20>
    windFrequency=<1.5 0.5 7.5>
    windOffset=<0 -150 50>
   }
   tactical_Sprint {
    maxDownInterpSpeed=5
    maxGustDuration=2
    maxGustStrength=7
    maxTimeBetweenGusts=7
    maxUpInterpSpeed=5
    minDownInterpSpeed=2
    minGustDuration=0.5
    minGustStrength=3
    minTimeBetweenGusts=0
    minUpInterpSpeed=2
    windTurbulenceScalarA=0.1
    windTurbulenceScalarB=1
    windAmplitude=<0 0 0>
    windFrequency=<0 0 0>
    windOffset=<0 0 0>
   }
   ziplining {
    maxDownInterpSpeed=5
    maxGustDuration=2
    maxGustStrength=7
    maxTimeBetweenGusts=7
    maxUpInterpSpeed=5
    minDownInterpSpeed=2
    minGustDuration=0.5
    minGustStrength=3
    minTimeBetweenGusts=0
    minUpInterpSpeed=2
    windTurbulenceScalarA=0.1
    windTurbulenceScalarB=1
    windAmplitude=<0 0 0>
    windFrequency=<0 0 0>
    windOffset=<0 0 0>
   }
  }
 }
]
