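# Post-process style hair AnimBP for the Ultralight female face accessory, layering body and parent poses before simulating dynamic hair.
# Head rotation and cape state drive additive hair and collider offsets feeding a RigidBody simulation, followed by distance-based scaling, crouch offsets and spring AnimDynamics on right tips.
# The event graph computes those alphas, sets gravity by movement state and adjusts acceleration scale for skydiving, crouching and backbling.
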AnimBlueprint /BRCosmetics/Characters/Player/Female/Medium/Bodies/F_MED_Ultralight/Meshes/Parts/F_MED_Ultralight_FaceAcc_AnimBP
target_skeleton=&/F_MED_Ultralight_FaceAcc_Skeleton
graphs [
 AnimGraph {nodes [
   Look At - Bone: dyn_hair_bk_MSTR(Alpha=1) {
    alphaBoolBlend {blendOption=Linear}
    boneToModify {boneName=dyn_hair_bk_MSTR}
    lookAtLocation=<0 -15.5>
    lookAtTarget {boneReference {boneName=spine_05}}
    lookAt_Axis {axis=<1 0>}
   }
   n1 = Copy Pose From Mesh(SourceMeshComponent=Get Body_Mesh())
   n2 = Copy Pose From Mesh() {bUseAttachedParent=true}
   n3 = Get Head_Rot_Bk() -> (Head_Rot_Bk:double)
   n4 = "Transform (Modify) Bone - Bone: dyn_hair_bk_MSTR"(
    ComponentPose=Local To Component(LocalPose=Layered blend per bone(BasePose=@n2, BlendPoses_0=@n1) {layerSetup [{branchFilters [{boneName=upperarm_l}, {boneName=upperarm_r}]}]})
    Alpha=@n3
    Translation="0,0,0"
    Rotation="25,0.000,0.000"
    Scale="1,1,1"
   ) {
    rotationMode=BMM_Additive
    rotationSpace=BCS_BoneSpace
    translationMode=BMM_Additive
    translationSpace=BCS_BoneSpace
    alphaBoolBlend {blendOption=Linear}
    boneToModify {boneName=dyn_hair_bk_MSTR}
    rotation {pitch=25}
   }
   n5 = "Transform (Modify) Bone - Bone: dyn_hair_r_MSTR"(ComponentPose=@n4, Alpha=@n3, Translation="0,1,0", Rotation="20,0,0", Scale="1,1,1") {
    rotationMode=BMM_Additive
    rotationSpace=BCS_BoneSpace
    translationMode=BMM_Additive
    translationSpace=BCS_BoneSpace
    alphaBoolBlend {blendOption=Linear}
    boneToModify {boneName=dyn_hair_r_MSTR}
    rotation {pitch=20}
    translation {y=1}
   }
   n6 = "Transform (Modify) Bone - Bone: dyn_hair_l_MSTR"(ComponentPose=@n5, Alpha=@n3, Translation="0,-1,0", Rotation="20,0,0", Scale="1,1,1") {
    rotationMode=BMM_Additive
    rotationSpace=BCS_BoneSpace
    translationMode=BMM_Additive
    translationSpace=BCS_BoneSpace
    alphaBoolBlend {blendOption=Linear}
    boneToModify {boneName=dyn_hair_l_MSTR}
    rotation {pitch=20}
    translation {y=-1}
   }
   n7 = Get Head_Rot_R() -> (Head_Rot_R:double)
   n8 = "Transform (Modify) Bone - Bone: dyn_hair_r_MSTR"(ComponentPose=@n6, Alpha=@n7, Translation="-5,3,5", Rotation="0,0,0", Scale="1,1,1") {
    rotationMode=BMM_Additive
    rotationSpace=BCS_BoneSpace
    translationMode=BMM_Additive
    translationSpace=BCS_BoneSpace
    alphaBoolBlend {blendOption=Linear}
    boneToModify {boneName=dyn_hair_r_MSTR}
    translation=<-5 3 5>
   }
   n9 = "Transform (Modify) Bone - Bone: dyn_hair_r_MSTR"(ComponentPose=@n8, Alpha=@n7, Translation="0,0,0", Rotation="0,-20,0", Scale="1,1,1") {
    rotationMode=BMM_Additive
    rotationSpace=BCS_BoneSpace
    translationMode=BMM_Additive
    translationSpace=BCS_BoneSpace
    alphaBoolBlend {blendOption=Linear}
    boneToModify {boneName=dyn_hair_r_MSTR}
    rotation {yaw=-20}
   }
   n10 = Get isWearCape() -> (isWearCape:bool)
   n11 = "Transform (Modify) Bone - Bone: dyn_hair_bk_MSTR"!(ComponentPose=@n9, bAlphaBoolEnabled:bool=@n10, Translation="0,0,0", Rotation="8,0.000,0.000", Scale="1,1,1") -> (Pose) {
    alphaInputType=Bool
    rotationMode=BMM_Additive
    rotationSpace=BCS_BoneSpace
    translationMode=BMM_Additive
    translationSpace=BCS_BoneSpace
    alphaBoolBlend {blendOption=Linear}
    boneToModify {boneName=dyn_hair_bk_MSTR}
    rotation {pitch=8}
   }
   n12 = "Transform (Modify) Bone - Bone: dyn_hair_r_MSTR"!(ComponentPose=@n11, bAlphaBoolEnabled:bool=@n10, Translation="0,0,0", Rotation="5,0,0", Scale="1,1,1") -> (Pose) {
    alphaInputType=Bool
    rotationMode=BMM_Additive
    rotationSpace=BCS_BoneSpace
    translationMode=BMM_Additive
    translationSpace=BCS_BoneSpace
    alphaBoolBlend {blendOption=Linear}
    boneToModify {boneName=dyn_hair_r_MSTR}
    rotation {pitch=5}
   }
   n13 = "Transform (Modify) Bone - Bone: dyn_hair_l_MSTR"!(ComponentPose=@n12, bAlphaBoolEnabled:bool=@n10, Translation="0,0,0", Rotation="5,0,0", Scale="1,1,1") -> (Pose) {
    alphaInputType=Bool
    rotationMode=BMM_Additive
    rotationSpace=BCS_BoneSpace
    translationMode=BMM_Additive
    translationSpace=BCS_BoneSpace
    alphaBoolBlend {blendOption=Linear}
    boneToModify {boneName=dyn_hair_l_MSTR}
    rotation {pitch=5}
   }
   n14 = "Transform (Modify) Bone - Bone: dyn_cape_collider"!(ComponentPose=@n13, bAlphaBoolEnabled:bool=@n10, Translation="-5.910,-14.689,2.149", Rotation="-0.004,-20.000,-0.000", Scale="1,1,1") -> (Pose) {
    alphaInputType=Bool
    rotationMode=BMM_Additive
    rotationSpace=BCS_BoneSpace
    translationMode=BMM_Additive
    translationSpace=BCS_BoneSpace
    alphaBoolBlend {blendOption=Linear}
    boneToModify {boneName=dyn_cape_collider}
    rotation=<rot -0.0039 -20 -0.0002>
    translation=<-5.91 -14.689 2.149>
   }
   n15 = RigidBody!(
    ComponentPose=@n14
    Alpha
    OverrideWorldGravity:Vector=Get GravityOverride()
    ExternalForce
    ComponentLinearAccScale:Vector=Get AccScaleOverride() -> (AccScaleOverride:Vector)
   ) -> (Pose) {
    bClampLinearTranslationLimitToRefPose=true
    bOverrideWorldGravity=true
    overridePhysicsAsset=&/F_MED_Ultralight_FaceAcc_Dyn_Hair
    simulationSpace=BaseBoneSpace
    alphaBoolBlend {blendOption=Linear}
    alphaScaleBiasClamp {interpSpeedDecreasing=4, interpSpeedIncreasing=6}
    baseBoneRef {boneName=spine_05}
    componentAppliedLinearAccClamp=<800 800 500>
    componentLinearAccScale=<0.6 0.6 0.2>
    componentLinearVelScale=<0.5 0.5 0.5>
    overrideWorldGravity {z=-875}
    simSpaceSettings {maxAngularAcceleration=200, maxAngularVelocity=5, maxLinearAcceleration=500, maxLinearVelocity=200}
   }
   n16 = Get hairScaleUpAlpha() -> (hairScaleUpAlpha:double)
   n17 = "Transform (Modify) Bone - Bone: dyn_hair_bk_02"(ComponentPose=@n15, Alpha=@n16, Translation="0,0,0", Rotation="0,0,0", Scale="1.000,1.000,1.5") {scaleMode=BMM_Additive, scaleSpace=BCS_BoneSpace, alphaBoolBlend {blendOption=Linear}, boneToModify {boneName=dyn_hair_bk_02}, scale {z=1.5}}
   n18 = "Transform (Modify) Bone - Bone: dyn_hair_l_03"(ComponentPose=@n17, Alpha=@n16, Translation="0,0,0", Rotation="0,0,0", Scale="1.000,1,1.75") {scaleMode=BMM_Additive, scaleSpace=BCS_BoneSpace, alphaBoolBlend {blendOption=Linear}, boneToModify {boneName=dyn_hair_l_03}, scale {z=1.75}}
   n19 = "Transform (Modify) Bone - Bone: dyn_hair_r_02"(ComponentPose=@n18, Alpha=@n16, Translation="0,0,0", Rotation="0,0,0", Scale="1.000,1,1.75") {scaleMode=BMM_Additive, scaleSpace=BCS_BoneSpace, alphaBoolBlend {blendOption=Linear}, boneToModify {boneName=dyn_hair_r_02}, scale {z=1.75}}
   n20 = Get bIsCrouching()
   n21 = "Transform (Modify) Bone - Bone: dyn_hair_r_02"!(ComponentPose=@n19, bAlphaBoolEnabled:bool=@n20, Translation="0,1,0", Rotation="0,0,0", Scale="1,1,1") -> (Pose) {
    alphaInputType=Bool
    rotationMode=BMM_Additive
    rotationSpace=BCS_BoneSpace
    translationMode=BMM_Additive
    translationSpace=BCS_BoneSpace
    alphaBoolBlend {blendInTime=0.2, blendOption=Linear, blendOutTime=0.3}
    boneToModify {boneName=dyn_hair_r_02}
    translation {y=1}
   }
   n22 = "Transform (Modify) Bone - Bone: dyn_hair_l_02"!(ComponentPose=@n21, bAlphaBoolEnabled:bool=@n20, Translation="0,-2,0", Rotation="0,0,0", Scale="1,1,1") -> (Pose) {
    alphaInputType=Bool
    rotationMode=BMM_Additive
    rotationSpace=BCS_BoneSpace
    translationMode=BMM_Additive
    translationSpace=BCS_BoneSpace
    alphaBoolBlend {blendInTime=0.2, blendOption=Linear, blendOutTime=0.3}
    boneToModify {boneName=dyn_hair_l_02}
    translation {y=-2}
   }
   n23 = Anim Dynamics - Bone: dyn_hairTips_r_01!(ComponentPose=@n22, Alpha=1, GravityScale:float=1, GravityOverride:Vector="0,0,-980", ExternalForce="0,0,0") -> (Pose) {
    angularDampingOverride=0.9999
    angularSpringConstant=1000
    bAngularSpring=true
    bGravityOverrideInSimSpace=true
    bOverrideAngularDamping=true
    simulationSpace=RootRelative
    alphaBoolBlend {blendOption=Linear}
    boundBone {boneName=dyn_hairTips_r_01}
    componentAppliedLinearAccClamp=<5000 5000 5000>
    componentLinearAccScale=<0.6 0.6 0.6>
    componentLinearVelScale=<0.6 0.6 0.6>
    gravityOverride {z=-980}
    physicsBodyDefinitions [{
      sphereCollisionRadius=0
      boundBone {boneName=dyn_hairTips_r_01}
      boxExtents=<12 10 10>
      constraintSetup {angularLimitsMax=<0 -10 15>, angularLimitsMin=<0 -30 -15>, angularTarget=<0 0 0>}
      localJointOffset=<1.5 0 -0.2>
     }]
    relativeSpaceBone {boneName=head}
   }
   n24 = Anim Dynamics - Bone: dyn_hairTips_r_02!(ComponentPose=@n23, Alpha=1, GravityScale:float=1, GravityOverride:Vector="0,0,-980", ExternalForce="0,0,0") -> (Pose) {
    angularDampingOverride=0.9999
    angularSpringConstant=1000
    bAngularSpring=true
    bGravityOverrideInSimSpace=true
    bOverrideAngularDamping=true
    simulationSpace=RootRelative
    alphaBoolBlend {blendOption=Linear}
    boundBone {boneName=dyn_hairTips_r_02}
    componentAppliedLinearAccClamp=<5000 5000 5000>
    componentLinearAccScale=<0.6 0.6 0.6>
    componentLinearVelScale=<0.6 0.6 0.6>
    gravityOverride {z=-980}
    physicsBodyDefinitions [{
      sphereCollisionRadius=0
      boundBone {boneName=dyn_hairTips_r_02}
      boxExtents=<12 10 10>
      constraintSetup {angularLimitsMax=<0 0 30>, angularLimitsMin=<0 -25 -30>, angularTarget=<0 0 0>}
      localJointOffset=<1 0 0>
     }]
    relativeSpaceBone {boneName=head}
   }
   n25 = Anim Dynamics - Bone: dyn_hairTips_r_03!(ComponentPose=@n24, Alpha=1, GravityScale:float=1, GravityOverride:Vector="0,0,-980", ExternalForce="0,0,0") -> (Pose) {
    angularDampingOverride=0.9999
    angularSpringConstant=1000
    bAngularSpring=true
    bGravityOverrideInSimSpace=true
    bOverrideAngularDamping=true
    simulationSpace=RootRelative
    alphaBoolBlend {blendOption=Linear}
    boundBone {boneName=dyn_hairTips_r_03}
    componentAppliedLinearAccClamp=<5000 5000 5000>
    componentLinearAccScale=<0.6 0.6 0.6>
    componentLinearVelScale=<0.6 0.6 0.6>
    gravityOverride {z=-980}
    physicsBodyDefinitions [{
      sphereCollisionRadius=0
      boundBone {boneName=dyn_hairTips_r_03}
      boxExtents=<12 10 10>
      constraintSetup {angularLimitsMax=<10 10 15>, angularLimitsMin=<-10 -20 -15>, angularTarget=<0 0 0>}
      localJointOffset=<1 0 0>
     }]
    relativeSpaceBone {boneName=head}
   }
   n26 = Anim Dynamics - Bone: dyn_hairTips_r_04!(ComponentPose=@n25, Alpha=1, GravityScale:float=1, GravityOverride:Vector="0,0,-980", ExternalForce="0,0,0") -> (Pose) {
    angularDampingOverride=1
    angularSpringConstant=1000
    bAngularSpring=true
    bGravityOverrideInSimSpace=true
    bOverrideAngularDamping=true
    simulationSpace=RootRelative
    alphaBoolBlend {blendOption=Linear}
    boundBone {boneName=dyn_hairTips_r_04}
    componentAppliedLinearAccClamp=<3000 3000 3000>
    componentLinearAccScale=<0.6 0.6 0.6>
    componentLinearVelScale=<0.6 0.6 0.6>
    gravityOverride {z=-980}
    physicsBodyDefinitions [{
      sphereCollisionRadius=0
      boundBone {boneName=dyn_hairTips_r_04}
      boxExtents=<14 12 10>
      constraintSetup {angularLimitsMax=<10 25 30>, angularLimitsMin=<-10 -25 -30>, angularTarget=<0 0 0>}
      localJointOffset=<3 0 0>
     }]
    relativeSpaceBone {boneName=head}
   }
   n27 = Anim Dynamics - Bone: dyn_hairTips_r_05!(ComponentPose=@n26, Alpha=1, GravityScale:float=1, GravityOverride:Vector="0,0,-980", ExternalForce="0,0,0") -> (Pose) {
    angularDampingOverride=1
    angularSpringConstant=1000
    bAngularSpring=true
    bGravityOverrideInSimSpace=true
    bOverrideAngularDamping=true
    simulationSpace=RootRelative
    alphaBoolBlend {blendOption=Linear}
    boundBone {boneName=dyn_hairTips_r_05}
    componentAppliedLinearAccClamp=<5000 5000 5000>
    componentLinearAccScale=<0.6 0.6 0.6>
    componentLinearVelScale=<0.6 0.6 0.6>
    gravityOverride {z=-980}
    physicsBodyDefinitions [{
      sphereCollisionRadius=0
      boundBone {boneName=dyn_hairTips_r_05}
      boxExtents=<12 10 10>
      constraintSetup {angularLimitsMax=<10 20 20>, angularLimitsMin=<-10 -20 -20>, angularTarget=<0 0 0>}
      localJointOffset=<1.5 0 0>
     }]
    relativeSpaceBone {boneName=head}
   }
   Output Pose(Result=Component To Local(ComponentPose=@n27)) {layerGroup=None}
  ]}
 EventGraph {nodes [
   Event BlueprintUpdateAnimation()
   > n1 = Sequence()
   n2 = GetOwnerSkeletalMeshComponent()
   n3 = BreakRotator(InRot=BreakTransform(InTransform=GetDeltaTransformFromRefPose(self=@n2, BoneName=head, BaseName=spine_05)).Rotation)
   > Set Head_Rot_Bk(Head_Rot_Bk:double=MapRangeClamped(Value=@n3.Yaw, InRangeA=-10, InRangeB=-55)) -> (Output_Get:double)
   > Set Head_Rot_R(Head_Rot_R:double=MapRangeClamped(Value=@n3.Pitch, InRangeA=0, InRangeB=45)) -> (Output_Get:double)
   n4 = GetPartSkeletalMeshComponent()
   > n5 = TransformFromBoneSpace!(
    execute
    self:Skinned Mesh Component Object Reference=@n4
    BoneName:Name=dyn_hair_bk_04
    InPosition:Vector="0, 0, 0"
    InRotation:Rotator="0, 0, 0"
   ) -> (then, OutPosition:Vector, OutRotation:Rotator)
   > n6 = TransformFromBoneSpace(self=@n4, BoneName=spine_05, InPosition="0, 0, 0", InRotation="0, 0, 0")
   > Set hairScaleUpAlpha(hairScaleUpAlpha:double=MapRangeClamped(Value="Distance (Vector)"!(V1:Vector=@n5, V2:Vector=@n6) -> (ReturnValue:double), InRangeA=12, InRangeB=40)) -> (Output_Get:double)
   n7 = Gravity Override State Function DEPRECATED(
    execute=@n1.then_1
    CharacterPartAnimInstance=Self-Reference()
    GameTimeInSeconds=GetGameTimeInSeconds()
    GravityOverrideParams=Get Gravity Override Params()
   )
   > Set GravityOverride(GravityOverride=@n7)
   > n8 = Branch(Condition=OR Boolean(A=Get bIsSkyDiving(), B=Get bIsCrouching()))
   > n9 = Set AccScaleOverride(AccScaleOverride:Vector="0.1,0.1,0.05") -> (Output_Get:Vector)
   n10 = Get isWearCape()
   n11 = Branch(execute=[@n8.else, @n9.then], Condition=AND Boolean(A=Get isWearBackbling() -> (isWearBackbling:bool), B=NOT Boolean(A=@n10)))
   > Set AccScaleOverride(AccScaleOverride="0.1,0.1,0.1")
   Set AccScaleOverride(execute=@n11.else, AccScaleOverride="0.3,0.3,0.25")
   Event OnCharacterCustomizationLoaded()
   > Do Once()
   n12 = CheckForBackbling()
   > Set isWearBackbling(isWearBackbling:bool=@n12) -> (Output_Get:bool)
   > Set isWearCape(isWearCape:bool=@n12.bIsBackblingLong) -> (Output_Get:bool)
   Event BlueprintInitializeAnimation()
   > Set Body_Mesh(Body_Mesh=GetOwnerSkeletalMeshComponent())
  ]}
]
variables [
 AccScaleOverride {type=Vector, default=<0.6 0.6 0.2>}
 Body_Mesh {type=SkeletalMeshComponent}
 Gravity Override Params {
  type=GravityOverrideParamsStruct
  default {
   dBNO {gravityOverride=<-700 200 0>}
   dBNO_Carried {gravityOverride=<-700 300 -300>}
   emoteOrMelee {gravityOverride=<0 0 -700>}
   front_End {gravityOverride=<0 0 -700>}
   onGround_Falling {gravityOverride=<0 0 -700>}
   onGround_Moving {gravityOverride=<0 0 -700>}
   onGround_Standing {gravityOverride=<0 0 -700>}
   skydive_DivingDown {gravityOverride=<-1500 600 0>, windAmplitude=<50 50 100>}
   skydive_FloatingDown {gravityOverride=<0 -100 -350>, windAmplitude=<50 -100 50>, windOffset=<0 0 50>}
   skydive_GoingUp {gravityOverride=<0 -100 -300>, windAmplitude=<50 -200 100>}
   skydive_Parachuting {gravityOverride=<0 -500 -700>, windAmplitude=<100 -150 200>}
   swimming {gravityOverride=<-500 200 0>}
   ziplining {gravityOverride=<0 0 -700>}
  }
 }
 GravityOverride {type=Vector}
 hairScaleUpAlpha {type=number}
 Head_Rot_Bk {type=number}
 Head_Rot_R {type=number}
 isWearBackbling {type=boolean}
 isWearCape {type=boolean}
]
